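; Train manager Blueprint that gives every car in a train one shared, randomly chosen color.
; On BeginPlay the authority picks ColorIndex from zero through the possible color count, and a custom event can force an index at runtime for debugging.
; Clients apply the color when ColorIndex replicates or local cars change, looping through valid BaseCar actors to set their mesh color index.
; The recolor pass runs only without authority.
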(event EventBeginPlay
  ; Randomize color index that all cars will use
  (Networking|SwitchHasAuthority
    (:Authority
      (Variables|Default|SetColorIndex (Math|Random|RandomInteger (+ (Variables|Default|GetPossibleColorCount) 1))))))

(event EventBP_OnLocalCarsChanged
  (CallFunction|UpdateAllCarColors))

(event Custom|SetTrainColorIndex
  (params (ColorIndex :type Integer))
  ; For forcing a specific train color at runtime.
  ; cheat ke* SetTrainColorIndex <index>
  (Variables|Default|SetColorIndex ColorIndex))

(fn OnRep_ColorIndex
  (Networking|SwitchHasAuthority
    (:Remote
      (CallFunction|UpdateAllCarColors))))

(fn UpdateAllCarColors
  (Networking|SwitchHasAuthority
    (:Remote
      (for _array_element (Variables|FortTrainManager|GetCars)
        (bind _asbase_car (Utilities|Casting|CastToBaseCar (Utilities|Struct|BreakTrainCarRuntimeData _array_element)))
        (Utilities|IsValid _asbase_car
          (:Is%20Valid
            (Class|BaseCar|SetMeshColorIndex _asbase_car (Variables|Default|GetColorIndex))))))))
