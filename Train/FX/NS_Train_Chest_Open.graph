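# Niagara burst effect for a train chest opening: a warm golden flash with rays, smoke and sparks.
# Combined_Emitter_Sprite packs three sprites, two meshes and a light into one emitter, animating per-renderer attributes with curves over normalized age.
# Rays, Smoke and Sparks are separate sprite bursts, with flipbook smoke and curl-noise sparks under gravity and drag.
# User.u_Scale sizes only the combined emitter's sprites and meshes; User.GlobalOffset shifts only that emitter and Smoke.
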NiagaraSystem /Train/FX/NS_Train_Chest_Open
dependencies {
 usedDataInterfaces [
  &NiagaraDataInterfaceCurve
  &NiagaraDataInterfaceVector2DCurve
  &NiagaraDataInterfaceVectorCurve
 ]
 usedDynamicInputs []
 usedModules [&SystemState]
}
emitters {
 Combined_Emitter_Sprite {
  emitterUpdateScript [
   &EmitterState()
   &SpawnBurst_Instantaneous(Use Spawn Probability=false, UseLoopCountLimit=false)
  ]
  particleSpawnScript [
   &InitializeParticle(
    Color Mode=::NewEnumerator0
    Position Offset=$User.GlobalOffset
    Sprite Rotation Mode=::NewEnumerator2
    UsePositionOffset=true
    Write Material Random=true
   )
   /SetVariables(
    Particles.Sprite_01_CameraOffset=100
    Particles.Sprite_01_Color:LinearColor=<rgba 2 1.1864 0.22 0.1>
    Particles.Sprite_01_Size:Vector2f=&Multiply_Vector2D_ByFloat(Float=$User.u_Scale, Vector2D=<200 200>)
   ) {isSetParameters=true}
   /SetVariables(
    Particles.Sprite_02_Color:LinearColor=<rgba 2 0.8 0.0667 1>
    Particles.Sprite_02_Size:Vector2f=&Multiply_Vector2D_ByFloat(Float=$User.u_Scale, Vector2D=&UniformRangedVector2D(Minimum=<750 750>, Maximum=<1000 1000>))
   ) {isSetParameters=true}
   /SetVariables(
    Particles.Sprite_03_Color:LinearColor=<rgba 30 23.0751 5.6771 0.15>
    Particles.Sprite_03_Size:Vector2f=&Vector2DFromFloat(Value=&Multiply_Float(A=$User.u_Scale, B=&RandomRangeFloat(Minimum=400, Maximum=500)))
   ) {isSetParameters=true}
   /SetVariables(
    Particles.Mesh_Color_Shape:LinearColor=<rgba 5 1.6247 0.4688 0.25>
    Particles.Mesh_DynamicParameter_Shape:Vector4f=&MakeVector4(X=1, Y=1, Z=1, W=&UniformRangedFloat(Minimum=0, Maximum=20))
    Particles.Mesh_Size_Shape:Vector3f=&VectorFromFloat(Value=&Multiply_Float(A=&RandomRangeFloat(Minimum=0.8, Maximum=1), B=$User.u_Scale))
   ) {isSetParameters=true}
   /SetVariables(
    Particles.Mesh_CameraOffset_Shape_02=50
    Particles.Mesh_Color_Shape_02:LinearColor=<rgba 15000 10299.4639 2343.75 0.2>
    Particles.Mesh_Size_Shape_02:Vector3f=&VectorFromFloat(Value=&Multiply_Float(A=0.3, B=$User.u_Scale))
   ) {isSetParameters=true}
  ]
  particleUpdateScript [
   &ParticleState()
   /SetVariables(
    Particles.Sprite_01_Color:LinearColor=&MultiplyLinearColor(
     Linear Color=$Particles.Initial.Sprite_01_Color
     Scale Factor=&MakeFloat4FromVectorAndFloat(
      "Vector (RGB)"=&VectorFromFloat(Value=&FloatFromCurve(
        FloatCurve=DataInterface() {propertyValues {Curve {keys [<at -0.0046 1.0113>,<at 1.0011 0.002>]}}}
        CurveIndex=&Divide_Float(A=$Particles.NormalizedAge, B=1)
       ))
      "Float (Alpha)"=&FloatFromCurve(
       FloatCurve=DataInterface() {propertyValues {Curve {keys [<at -0.0046 1.0113>,<at 1.0011 0.002>]}}}
       CurveIndex=&Divide_Float(A=$Particles.NormalizedAge, B=1)
      )
     )
    )
    Particles.Sprite_01_Size:Vector2f=&Multiply_Vector2D(
     A=$Particles.Initial.Sprite_01_Size
     B=&Vector2DFromCurve(
      Vector2Curve=DataInterface() {propertyValues {
        XCurve {keys [
          {arriveTangent=4.1785, interpMode=RCIM_Cubic, leaveTangent=4.1785, tangentMode=RCTM_User, time=0, value=0}
          {arriveTangent=1.1179, interpMode=RCIM_Cubic, leaveTangent=1.1179, tangentMode=RCTM_User, time=0.2042, value=0.6228}
          {arriveTangent=0.0224, interpMode=RCIM_Cubic, leaveTangent=0.0224, tangentMode=RCTM_User, time=1, value=1}
         ]}
        YCurve {keys [
          {arriveTangent=4.1785, interpMode=RCIM_Cubic, leaveTangent=4.1785, tangentMode=RCTM_User, time=0, value=0}
          {arriveTangent=1.1179, interpMode=RCIM_Cubic, leaveTangent=1.1179, tangentMode=RCTM_User, time=0.2042, value=0.6228}
          {arriveTangent=0.0224, interpMode=RCIM_Cubic, leaveTangent=0.0224, tangentMode=RCTM_User, time=1, value=1}
         ]}
       }}
      CurveIndex=&Divide_Float(A=$Particles.NormalizedAge, B=1)
      ScaleCurve=<1 1>
     )
    )
   ) {isSetParameters=true}
   /SetVariables(
    Particles.Sprite_02_Color:LinearColor=&MultiplyLinearColor(
     Linear Color=$Particles.Initial.Sprite_01_Color
     Scale Factor=&MakeFloat4FromVectorAndFloat(
      "Vector (RGB)"=&VectorFromFloat(Value=&FloatFromCurve(
        FloatCurve=DataInterface() {propertyValues {Curve {keys [<at -0.0046 1.0113>,<at 1.0011 0.002>]}}}
        CurveIndex=&Divide_Float(A=$Particles.NormalizedAge, B=0.35)
       ))
      "Float (Alpha)"=&FloatFromCurve(FloatCurve=DataInterface() {propertyValues {Curve {keys [{arriveTangent=-3.3901, interpMode=RCIM_Cubic, leaveTangent=-3.3901, tangentMode=RCTM_User, time=-0.0046, value=1.0113}, <at 1.0011 0.002>]}}})
     )
    )
    Particles.Sprite_02_Size:Vector2f=&Multiply_Vector2D(
     A=$Particles.Initial.Sprite_01_Size
     B=&Vector2DFromCurve(
      Vector2Curve=DataInterface() {propertyValues {
        XCurve {keys [
          {arriveTangent=17.0743, interpMode=RCIM_Cubic, leaveTangent=17.0743, tangentMode=RCTM_User, time=0, value=0}
          {arriveTangent=-1.1985, interpMode=RCIM_Cubic, leaveTangent=-1.1985, tangentMode=RCTM_User, time=0.3195, value=1.6986}
          {arriveTangent=0.0224, interpMode=RCIM_Cubic, leaveTangent=0.0224, tangentMode=RCTM_User, time=1, value=1}
         ]}
        YCurve {keys [
          {arriveTangent=17.0743, interpMode=RCIM_Cubic, leaveTangent=17.0743, tangentMode=RCTM_User, time=0, value=0}
          {arriveTangent=-1.1985, interpMode=RCIM_Cubic, leaveTangent=-1.1985, tangentMode=RCTM_User, time=0.3195, value=1.6986}
          {arriveTangent=0.0224, interpMode=RCIM_Cubic, leaveTangent=0.0224, tangentMode=RCTM_User, time=1, value=1}
         ]}
       }}
      CurveIndex=&Divide_Float(A=$Particles.NormalizedAge, B=0.35)
      ScaleCurve=<1.5 1.5>
     )
    )
   ) {isSetParameters=true}
   /SetVariables(
    Particles.Sprite_03_Color:LinearColor=&MultiplyLinearColor(
     Linear Color=$Particles.Initial.Sprite_03_Color
     Scale Factor=&MakeFloat4FromVectorAndFloat(
      "Vector (RGB)"=&VectorFromFloat(Value=&FloatFromCurve(
        FloatCurve=DataInterface() {propertyValues {Curve {keys [<at -0.0046 1.0113>,<at 1.0011 0.002>]}}}
        CurveIndex=&Divide_Float(A=$Particles.NormalizedAge, B=0.4)
       ))
      "Float (Alpha)"=&FloatFromCurve(
       FloatCurve=DataInterface() {propertyValues {Curve {keys [<at -0.0046 1.0113>,<at 1.0011 0.002>]}}}
       CurveIndex=&Divide_Float(A=$Particles.NormalizedAge, B=0.4)
      )
     )
    )
    Particles.Sprite_03_DynamicParameter:Vector4f=&MakeVector4(
     X=&FloatFromCurve(
      FloatCurve=DataInterface() {propertyValues {Curve {keys [<at 0.7263 1.5005>,<at 1 1>]}}}
      CurveIndex=&Divide_Float(A=$Particles.NormalizedAge, B=0.4)
     )
     Y=2
     Z=1
     W=1
    )
    Particles.Sprite_03_Size:Vector2f=&Multiply_Vector2D(
     A=$Particles.Initial.Sprite_03_Size
     B=&Vector2DFromCurve(
      Vector2Curve=DataInterface() {propertyValues {XCurve {keys [<at -0.0107 0.2143>,<at 1.0179 0.6584>]}, YCurve {keys [<at -0.0269 0.1988>,<at 1.0179 0.6584>]}}}
      CurveIndex=&Divide_Float(A=$Particles.NormalizedAge, B=0.4)
      ScaleCurve=<1 1>
     )
    )
   ) {isSetParameters=true}
   /SetVariables(
    Particles.Mesh_Color_Shape:LinearColor=&MultiplyLinearColor(
     Linear Color=$Particles.Initial.Mesh_Color_Shape
     Scale Factor=&MakeFloat4FromVectorAndFloat(
      "Vector (RGB)"=&VectorFromCurve(
       VectorCurve=DataInterface() {propertyValues {
         XCurve {keys [<at -0.0011 1.0067>,<at 0.9988 0>]}
         YCurve {keys [<at -0.0011 1.0067>,<at 0.9988 0>]}
         ZCurve {keys [<at -0.0011 1.0067>,<at 0.9988 0>]}
        }}
       CurveIndex=&Divide_Float(A=$Particles.NormalizedAge, B=0.35)
      )
      "Float (Alpha)"=&FloatFromCurve(
       FloatCurve=DataInterface() {propertyValues {Curve {keys [
           {interpMode=RCIM_Cubic, tangentMode=RCTM_User, time=-0.0046, value=1.0113}
           {arriveTangent=-1.0036, interpMode=RCIM_Cubic, leaveTangent=-1.0036, tangentMode=RCTM_User, time=0.7114, value=0.7365}
           {arriveTangent=-2.4196, interpMode=RCIM_Cubic, leaveTangent=-2.4196, tangentMode=RCTM_User, time=1.0011, value=0.002}
          ]}}}
       CurveIndex=&Divide_Float(A=$Particles.NormalizedAge, B=0.35)
      )
     )
    )
    Particles.Mesh_DynamicParameter_Shape_01:Vector4f=&MakeVector4(
     X=&FloatFromCurve(
      FloatCurve=DataInterface() {propertyValues {Curve {keys [<at 0 0>,<at 0.22 0.5>,<at 1 0>]}}}
      CurveIndex=&Divide_Float(A=$Particles.NormalizedAge, B=0.35)
     )
     Y=&FloatFromCurve(
      FloatCurve=DataInterface() {propertyValues {Curve {keys [<at 0 15>,<at 1 0>]}}}
      CurveIndex=&Divide_Float(A=$Particles.NormalizedAge, B=0.35)
     )
     Z=0
     W=0
    )
    Particles.Mesh_Size_Shape:Vector3f=&Multiply_Vector(
     A=$Particles.Initial.Mesh_Size_Shape
     B=&VectorFromFloat(Value=&FloatFromCurve(
       FloatCurve=DataInterface() {propertyValues {Curve {keys [
           {arriveTangent=4.7007, interpMode=RCIM_Cubic, leaveTangent=4.7007, tangentMode=RCTM_User, time=0, value=0}
           {arriveTangent=2, arriveTangentWeight=0.0833, interpMode=RCIM_Cubic, leaveTangent=2, tangentMode=RCTM_User, time=0.25, value=1}
           {arriveTangentWeight=0.25, interpMode=RCIM_Cubic, tangentMode=RCTM_User, time=1, value=2}
          ]}}}
       CurveIndex=&Divide_Float(A=$Particles.NormalizedAge, B=0.35)
      ))
    )
   ) {isSetParameters=true}
   /SetVariables(
    Particles.Mesh_Color_Shape_02:LinearColor=&MultiplyLinearColor(
     Linear Color=$Particles.Initial.Mesh_Color_Shape_02
     Scale Factor=&MakeFloat4FromVectorAndFloat(
      "Vector (RGB)"=&VectorFromCurve(
       VectorCurve=DataInterface() {propertyValues {
         XCurve {keys [<at 0.0013 1.998>,<at 1 1>]}
         YCurve {keys [<at 0.0012 1.4636>,<at 1.0017 0.264>]}
         ZCurve {keys [<at 0.0012 1.0048>,<at 1.0017 0.264>]}
        }}
       CurveIndex=&Divide_Float(A=$Particles.NormalizedAge, B=&RandomRangeFloat(Minimum=0.6, Maximum=0.8))
      )
      "Float (Alpha)"=&FloatFromCurve(
       FloatCurve=DataInterface() {propertyValues {Curve {keys [
           {interpMode=RCIM_Cubic, tangentMode=RCTM_User, time=-0.0034, value=0.1203}
           {arriveTangent=-0.1606, interpMode=RCIM_Cubic, leaveTangent=-0.1606, tangentMode=RCTM_User, time=0.3345, value=0.0495}
           {interpMode=RCIM_Cubic, tangentMode=RCTM_User, time=0.9972, value=-0.0001}
          ]}}}
       CurveIndex=&Divide_Float(A=$Particles.NormalizedAge, B=&RandomRangeFloat(Minimum=0.6, Maximum=0.8))
      )
     )
    )
    Particles.Mesh_DynamicParameter_Shape_02:Vector4f=&MakeVector4(
     X=&FloatFromCurve(
      FloatCurve=DataInterface() {propertyValues {Curve {keys [
          {arriveTangent=11.0061, interpMode=RCIM_Cubic, leaveTangent=11.0061, tangentMode=RCTM_User, time=-0.0005, value=-0.0141}
          {arriveTangent=1.9873, interpMode=RCIM_Cubic, leaveTangent=1.9873, tangentMode=RCTM_User, time=0.1692, value=1.0698}
          <at 1.0021 1.9784>
          <at 1.61 1.9784>
         ]}}}
      CurveIndex=&Divide_Float(A=$Particles.NormalizedAge, B=&RandomRangeFloat(Minimum=0.6, Maximum=0.8))
      Scale Curve=0.25
     )
     Y=&FloatFromCurve(
      FloatCurve=DataInterface() {propertyValues {Curve {keys [<at -0.0005 15>,<at 0.1612 13.3874>,<at 1.0021 5>,<at 1.61 5>]}}}
      CurveIndex=&Divide_Float(A=$Particles.NormalizedAge, B=&RandomRangeFloat(Minimum=0.6, Maximum=0.8))
     )
     Z=-0.8
     W=0
    )
    Particles.Mesh_Size_Shape_02:Vector3f=&Multiply_Vector(
     A=$Particles.Initial.Mesh_Size_Shape_02
     B=&VectorFromCurve(
      VectorCurve=DataInterface() {propertyValues {
        XCurve {keys [
          {arriveTangent=9.7102, interpMode=RCIM_Cubic, leaveTangent=9.7102, tangentMode=RCTM_User, time=-0.0005, value=3.4203}
          {arriveTangent=6.8631, interpMode=RCIM_Cubic, leaveTangent=6.8631, tangentMode=RCTM_User, time=0.2646, value=6.6389}
          <at 0.9996 7.8749>
          <at 1.61 7.8749>
         ]}
        YCurve {keys [
          {arriveTangent=25.6031, interpMode=RCIM_Cubic, leaveTangent=25.6031, tangentMode=RCTM_User, time=-0.0005, value=3.4203}
          {arriveTangent=6.8631, interpMode=RCIM_Cubic, leaveTangent=6.8631, tangentMode=RCTM_User, time=0.2633, value=6.6865}
          <at 0.9996 7.8749>
          <at 1.61 7.8749>
         ]}
        ZCurve {keys [
          {arriveTangent=25.8257, interpMode=RCIM_Cubic, leaveTangent=25.8257, tangentMode=RCTM_User, time=-0.0005, value=3.4203}
          {arriveTangent=6.8631, interpMode=RCIM_Cubic, leaveTangent=6.8631, tangentMode=RCTM_User, time=0.2666, value=6.7585}
          <at 0.9996 7.8749>
          <at 1.61 7.8749>
         ]}
       }}
      CurveIndex=&Divide_Float(A=$Particles.NormalizedAge, B=&RandomRangeFloat(Minimum=0.6, Maximum=0.8))
     )
    )
   ) {isSetParameters=true}
   &FN_Light_Attributes(
    Color=<rgba 100 49.375 15.625 1>
    Create Color Attribute=true
    Create Exponent Attribute=true
    Create Position Attribute=true
    Create Radius Attribute=true
    Exponent=5
    Radius=100
    Scale Color=&FloatFromCurve(FloatCurve=DataInterface() {propertyValues {Curve {keys [{arriveTangent=-3.3901, interpMode=RCIM_Cubic, leaveTangent=-3.3901, tangentMode=RCTM_User, time=-0.0046, value=1.0113}, <at 1.0011 0>]}}})
   )
   &SolveForcesAndVelocity(Clamp Velocity=false, Limit Acceleration=false)
  ]
  rendererClasses [
   &NiagaraLightRendererProperties
   &NiagaraSpriteRendererProperties
   &NiagaraMeshRendererProperties
  ]
  renderers [
   {rendererClass=&NiagaraLightRendererProperties, rendererIndex=0}
   {rendererClass=&NiagaraSpriteRendererProperties, rendererIndex=1}
   {rendererClass=&NiagaraSpriteRendererProperties, rendererIndex=2}
   {rendererClass=&NiagaraSpriteRendererProperties, rendererIndex=3}
   {rendererClass=&NiagaraMeshRendererProperties, rendererIndex=4}
   {rendererClass=&NiagaraMeshRendererProperties, rendererIndex=5}
  ]
 }
 Rays {
  emitterUpdateScript [
   &EmitterState()
   &SpawnBurst_Instantaneous(Spawn Count=4, Use Spawn Probability=false, UseLoopCountLimit=false)
  ]
  particleSpawnScript [
   &InitializeParticle(
    Color=<rgba 3 1.4774 0.5469 0.1>
    Lifetime Max=1
    Lifetime Min=0.8
    Lifetime Mode=::NewEnumerator1
    Sprite Size Max=<200 200>
    Sprite Size Min=<150 150>
    Sprite Size Mode=::NewEnumerator2
    UsePositionOffset=false
    Write Material Random=true
   )
   /Niagara/Modules/Update/SubUV/SubUVAnimation()
   &ShapeLocation(
    Box Size=<25 10 20>
    MaskSpawnGroup=false
    Offset=<0 0 -25>
    Offset Mode=::NewEnumerator0
    Override Seed=true
    Random Seed=2147483647
    Randomness Mode:ENiagaraRandomnessMode=::NewEnumerator1
    Shape Primitive=::NewEnumerator2
    Surface Only=false
    UseOwnerScale=true
   )
   &AddVelocity(Velocity=<0 0 1>)
   &CameraOffset(Camera Offset Amount=50)
  ]
  particleUpdateScript [
   &ParticleState()
   &ScaleSpriteSize(Scale Factor=&Vector2DFromFloat(Value=&FloatFromCurve(FloatCurve=DataInterface() {propertyValues {Curve {keys [
          <at -0.0017 0.0417 RCIM_Cubic>
          {arriveTangent=1.1486, interpMode=RCIM_Cubic, leaveTangent=1.1486, tangentMode=RCTM_User, time=0.0317, value=1.2332}
          <at 0.57 1.2568>
          {arriveTangent=-0.1439, interpMode=RCIM_Cubic, leaveTangent=-0.1439, tangentMode=RCTM_User, time=0.9845, value=1.0006}
          {interpMode=RCIM_Cubic, tangentMode=RCTM_User, time=1.0011, value=0.002}
         ]}}})))
   &ScaleColor(
    Scale Alpha=&FloatFromCurve(FloatCurve=DataInterface() {propertyValues {Curve {keys [<at -0.0046 1.0113 RCIM_Cubic>,<at 1.0011 0.002 RCIM_Cubic>]}}})
    ScaleA=true
    ScaleRGB=true
   )
   &SolveForcesAndVelocity(Clamp Velocity=false, Limit Acceleration=false)
  ]
  rendererClasses [&NiagaraSpriteRendererProperties]
  renderers [{rendererClass=&NiagaraSpriteRendererProperties}]
 }
 Smoke {
  emitterUpdateScript [
   &EmitterState()
   &SpawnBurst_Instantaneous(Spawn Count=5, Use Spawn Probability=false, UseLoopCountLimit=false)
  ]
  particleSpawnScript [
   &InitializeParticle(
    Color=<rgba 1 0.7409 0.302 1>
    Lifetime Mode=::NewEnumerator1
    Mass Max=2
    Mass Mode=::NewEnumerator3
    Position Offset=$User.GlobalOffset
    Sprite Rotation Mode=::NewEnumerator2
    Sprite Size Mode=::NewEnumerator1
    Uniform Sprite Size Max=175
    Uniform Sprite Size Min=100
    UsePositionOffset=true
    Write Material Random=true
   )
   &ShapeLocation(
    Box Size=<25 25 25>
    MaskSpawnGroup=false
    Offset=<0 0 -10>
    Offset Mode=::NewEnumerator0
    Override Seed=false
    Shape Primitive=::NewEnumerator2
    Surface Only=false
    UseOwnerScale=true
   )
   &AddVelocity(Velocity=<0 -25 50>)
   &GravityForce(Gravity=<0 0 -20>)
   &ApplyInitialForces()
  ]
  particleUpdateScript [
   &ParticleState()
   &ScaleSpriteSize(Scale Factor=&Vector2DFromFloat(Value=&FloatFromCurve(FloatCurve=DataInterface() {propertyValues {Curve {keys [
          {arriveTangent=2.9149, interpMode=RCIM_Cubic, leaveTangent=2.9149, tangentMode=RCTM_User, time=0.0011, value=0.4546}
          {arriveTangent=0.9127, interpMode=RCIM_Cubic, leaveTangent=0.9127, tangentMode=RCTM_User, time=0.0742, value=0.7417}
          <at 1 1>
         ]}}})))
   &ScaleColor(
    Scale Alpha=&FloatFromCurve(FloatCurve=DataInterface() {propertyValues {Curve {keys [
         {interpMode=RCIM_Cubic, tangentMode=RCTM_User, time=0.0007, value=0.9747}
         {arriveTangent=-0.9969, interpMode=RCIM_Cubic, leaveTangent=-0.9969, tangentMode=RCTM_User, time=0.1357, value=0.3063}
         {interpMode=RCIM_Cubic, tangentMode=RCTM_User, time=1.0011, value=0.002}
        ]}}})
    Scale RGB=&VectorFromCurve(VectorCurve=DataInterface() {propertyValues {
       XCurve {keys [
         {arriveTangent=-36.1644, interpMode=RCIM_Cubic, leaveTangent=-36.1645, tangentMode=RCTM_User, time=0, value=21.1927}
         {arriveTangent=-18.5517, interpMode=RCIM_Cubic, leaveTangent=-18.5516, tangentMode=RCTM_User, time=0.1272, value=8.8294}
         {interpMode=RCIM_Cubic, tangentMode=RCTM_User, time=0.9988, value=0}
        ]}
       YCurve {keys [
         {arriveTangent=-117.4017, interpMode=RCIM_Cubic, leaveTangent=-117.413, tangentMode=RCTM_User, time=0, value=15.1453}
         {arriveTangent=-13.8124, interpMode=RCIM_Cubic, leaveTangent=-13.8124, tangentMode=RCTM_User, time=0.1159, value=6.8365}
         {interpMode=RCIM_Cubic, tangentMode=RCTM_User, time=0.9988, value=0}
        ]}
       ZCurve {keys [
         {arriveTangent=-95.6646, interpMode=RCIM_Cubic, leaveTangent=-95.6645, tangentMode=RCTM_User, time=0, value=8.6855}
         {arriveTangent=-10.2211, interpMode=RCIM_Cubic, leaveTangent=-10.2214, tangentMode=RCTM_User, time=0.1015, value=5.1184}
         {interpMode=RCIM_Cubic, tangentMode=RCTM_User, time=0.9988, value=0}
        ]}
      }})
    ScaleA=true
    ScaleRGB=true
   )
   &SolveForcesAndVelocity(Clamp Velocity=false, Limit Acceleration=false)
   &DynamicMaterialParameters(
    Index 0 Param 1=1
    Module.Param0WriteEnabled=true
    Module.Param1WriteEnabled=true
    Module.Param2WriteEnabled=true
    Module.Param3WriteEnabled=true
   )
   /SetVariables(Particles.SubImageIndex=&Lerp_Float(A=24, B=63, Alpha=$Particles.NormalizedAge)) {isSetParameters=true}
  ]
  rendererClasses [&NiagaraSpriteRendererProperties]
  renderers [{rendererClass=&NiagaraSpriteRendererProperties}]
 }
 Sparks {
  emitterUpdateScript [
   &EmitterState(Life Cycle Mode=::NewEnumerator1, Loop Behavior=::NewEnumerator1, Loop Duration=1, UseLoopDelay=false)
   &SpawnBurst_Instantaneous(Spawn Count=17, Use Spawn Probability=false, UseLoopCountLimit=false)
  ]
  particleSpawnScript [
   &InitializeParticle(
    Color=<rgba 5 3.7043 1.51 1>
    Lifetime Max=1
    Lifetime Min=0.1
    Lifetime Mode=::NewEnumerator1
    Sprite Size Mode=::NewEnumerator1
    Uniform Sprite Size Max=30
    Uniform Sprite Size Min=10
    UsePositionOffset=false
    Write Material Random=true
   )
   &ShapeLocation(Box Size=<60 25 60>, MaskSpawnGroup=false, Override Seed=false, Shape Primitive=::NewEnumerator2, Surface Only=false, UseOwnerScale=true)
   &AddVelocity(
    Velocity=&RandomRangeVector(Minimum=<-0.25 0 1>, Maximum=<0.25 0.2 1>, Randomness Mode=::NewEnumerator1, Random Seed=5)
    Velocity Speed Scale=&RandomRangeFloat(Minimum=100, Maximum=250, Randomness Mode=::NewEnumerator1, Random Seed=5)
   )
  ]
  particleUpdateScript [
   &ParticleState()
   &GravityForce(Gravity=<0 0 -50>)
   &CurlNoiseForce(Noise Frequency=5, Noise Strength=100, Pan Noise=false, Random Seed=5)
   &Drag(
    Drag=&RandomRangeFloat(Minimum=1.5, Maximum=3, Randomness Mode=::NewEnumerator1, Random Seed=5)
    Use Linear Drag=true
    Use Rotational Drag=false
   )
   &ScaleColor(
    Scale Alpha=&FloatFromCurve(
     FloatCurve=DataInterface() {propertyValues {Curve {keys [<at -0.0092 0.0144>,<at 0.0967 0.9767>,<at 1.0011 0.002>]}}}
     Scale Curve=&RandomRangeFloat(Minimum=0.25, Maximum=0.75)
    )
    ScaleA=true
    ScaleRGB=true
   )
   &ScaleSpriteSize(Scale Factor=&Vector2DFromFloat(Value=&FloatFromCurve(FloatCurve=DataInterface() {propertyValues {Curve {keys [<at -0.0046 0>,<at 0.1746 0.9952>,<at 1.0011 0.002>]}}})))
   &SolveForcesAndVelocity(Clamp Velocity=false, Limit Acceleration=false)
  ]
  rendererClasses [&NiagaraSpriteRendererProperties]
  renderers [{rendererClass=&NiagaraSpriteRendererProperties}]
 }
}
summary {
 systemName=NS_Train_Chest_Open
 emitters [
  {
   emitterName=Combined_Emitter_Sprite
   rendererClasses [
    &NiagaraLightRendererProperties
    &NiagaraSpriteRendererProperties
    &NiagaraMeshRendererProperties
   ]
  }
  {emitterName=Rays, rendererClasses [&NiagaraSpriteRendererProperties]}
  {emitterName=Smoke, rendererClasses [&NiagaraSpriteRendererProperties]}
  {emitterName=Sparks, rendererClasses [&NiagaraSpriteRendererProperties]}
 ]
 userVariables [
  User.GlobalOffset {type=Vector3f, description="", defaultValue {struct=Vector3f, value=<0 0 0>}}
  User.u_Scale {type=NiagaraFloat, defaultValue=1, description=""}
 ]
}
