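# Niagara effect for a successful train graffiti capture box, puffing smoke colored from a user color array and tinted by time of day.
# Only Smoke is enabled: a burst with upward kick plus two curve-driven spawn rates, tuned by Delay, Speed and Scale_Rate.
# Spawn position sweeps along a scaled torus over time, pushing puffs along its normal.
# Glow, gloop, splash and GPU 2D-grid fluid emitters (fed by Source) remain but are disabled.
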NiagaraSystem /Train/FX/NS_Train_Graffiti_Success_Capture_Box
dependencies {
 usedDataInterfaces [
  &NiagaraDataInterfaceCurve
  &NiagaraDataInterfaceArrayFloat
  &NiagaraDataInterfaceSpriteRendererInfo
  /Script/Niagara.NiagaraDataInterfaceGrid2DCollection
  /Script/Niagara.NiagaraDataInterfaceRenderTarget2D
  &NiagaraDataInterfaceCamera
  &NiagaraDataInterfacePlatformSet
  &NiagaraDataInterfaceParticleRead
  /Script/Niagara.NiagaraDataInterfaceEmitterProperties
  &NiagaraDataInterfaceVector2DCurve
  &NiagaraDataInterfaceColorCurve
  &NiagaraDataInterfaceVectorCurve
 ]
 usedDynamicInputs [&Vector2DFromCurve]
 usedModules [
  &SystemState
  &self:SystemSpawnScript.SystemScriptSource.SystemScriptGraph.NiagaraNodeAssignment_0.SetVariables_#1
 ]
}
emitters {
 Glow {
  bEnabled=false
  emitterUpdateScript [
   &EmitterState(Life Cycle Mode=::NewEnumerator1, Loop Behavior=::NewEnumerator1, Loop Delay=$User.Delay, Loop Duration=1, UseLoopDelay=true)
   &SpawnRate(
    SpawnRate=&FloatFromCurve(
     FloatCurve=DataInterface() {propertyValues {Curve {keys [<at 0 1>,<at 2.5 0>]}}}
     CurveIndex=&Subtract_Float(A=$System.Age, B=$User.Delay)
     Scale Curve=20
    )
    Use Spawn Probability=false
   )
  ]
  particleSpawnScript [
   &FN_SetSpawnGroup(Number of Spawn Groups=2)
   &InitializeParticle(
    Color=<rgba 0.5 0.5 0.5 0.2>
    Lifetime=2
    Sprite Size Mode=::NewEnumerator3
    Uniform Sprite Size=1000
    UsePositionOffset=false
    Write Material Random=true
   )
   &ShapeLocation(
    Handle Radius=4
    MaskSpawnGroup=false
    Non Uniform Scale=<0 18 30>
    Offset=<0 100 0>
    Offset Mode=::NewEnumerator0
    Rotation Mode=::NewEnumerator2
    Shape Primitive=::NewEnumerator3
    Torus Distribution Mode:ENiagaraTorusDistributionMode=::NewEnumerator1
    U Position=&MaskFloatBySpawnGroup(
     Masked Float=&FloatFromCurve(
      FloatCurve=DataInterface() {propertyValues {Curve {keys [<at 0 0.5>,<at 1 0.55>]}}}
      CurveIndex=&Subtract_Float(A=&Multiply_Float(A=$System.Age, B=1), B=$User.Delay)
     )
     Passthrough Float=&FloatFromCurve(
      FloatCurve=DataInterface() {propertyValues {Curve {keys [<at 0 0.5>,<at 2.5 1>]}}}
      CurveIndex=&Subtract_Float(A=$System.Age, B=$User.Delay)
      Scale Curve=-1
     )
    )
    UseOwnerScale=true
    V Position=&RandomRangeFloat(Minimum=0.25, Maximum=0.75)
    Yaw / Pitch / Roll=<0 0.25 0.25>
   ) {module=ShapeLocation001}
   &CameraOffset(Camera Offset Amount=50)
  ]
  particleUpdateScript [
   &ParticleState()
   &ScaleColor(
    Scale Alpha=&FloatFromCurve(
     FloatCurve=DataInterface() {propertyValues {Curve {keys [
         {interpMode=RCIM_Cubic, tangentMode=RCTM_User, time=0, value=0}
         {interpMode=RCIM_Cubic, tangentMode=RCTM_User, time=0.25, value=1}
         {tangentMode=RCTM_User, time=1, value=0}
        ]}}}
     CurveIndex=&Subtract_Float(A=$System.Age, B=$User.Delay)
    )
    Scale RGB=&MakeVectorFromLinearColorRGB(Color=&ColorFromCurve(ColorCurve=DataInterface() {propertyValues {
        AlphaCurve {keys [{arriveTangent=0.9905, interpMode=RCIM_Cubic, time=0, value=1}]}
        BlueCurve {keys [<at 0 1>, {arriveTangent=-1, interpMode=RCIM_Cubic, leaveTangent=-1, time=0.0623, value=1}, <at 0.7214 0.2188>, <at 1 0>]}
        GreenCurve {keys [<at 0 0.6401>, {arriveTangent=-0.6401, interpMode=RCIM_Cubic, leaveTangent=-0.6401, time=0.0623, value=0.3855}, <at 0.7214 0>, <at 1 0>]}
        RedCurve {keys [<at 0 0>,<at 0.0623 0 RCIM_Cubic>,<at 0.7214 0.1373>,<at 1 0>]}
       }}))
    ScaleA=true
    ScaleRGB=true
   )
  ]
  rendererClasses [&NiagaraSpriteRendererProperties]
  renderers [{rendererClass=&NiagaraSpriteRendererProperties}]
 }
 Main_Gloop {
  bEnabled=false
  emitterUpdateScript [
   &EmitterState(
    Delay First Loop Only=true
    Enable Distance Culling=true
    Life Cycle Mode=::NewEnumerator1
    Loop Behavior=::NewEnumerator2
    Loop Count=6
    Loop Delay=$User.Delay
    Loop Duration=0.25
    MaxDistance=7000
    Module.bUseMaxDistance=true
    Module.bUseMinDistance=false
    Scalability Mode=::NewEnumerator1
    UseLoopDelay=true
   )
   &SpawnRate(
    SpawnRate=&FloatFromCurve(
     FloatCurve=DataInterface() {propertyValues {Curve {keys [<at 0 1>,<at 0.5 0.35>,<at 0.75 0>]}}}
     CurveIndex=&Subtract_Float(A=&Multiply_Float(A=$System.Age, B=$User.Speed), B=$User.Delay)
     Scale Curve=50
    )
    Use Spawn Probability=false
   ) {module=SpawnRate001}
   &SpawnRate(
    SpawnGroup=1
    SpawnRate=&FloatFromCurve(
     FloatCurve=DataInterface() {propertyValues {Curve {keys [<at 0 1>,<at 1.25 0>]}}}
     CurveIndex=&Subtract_Float(A=&Multiply_Float(A=$System.Age, B=$User.Speed), B=$User.Delay)
     Scale Curve=75
    )
    Use Spawn Probability=false
   ) {module=SpawnRate002}
  ]
  particleSpawnScript [
   /SetVariables(Particles.SpawnGroup=$Emitter.SpawnGroup) {isSetParameters=true}
   &InitializeParticle(
    Color=&SelectLinearColorFromArray(Array Sampling Mode=::NewEnumerator0, Direct Array Index=?, Color Selection Array=$User.Graffiti_Color_Array, Random Seed=0)
    Lifetime Max=0.5
    Lifetime Min=0.35
    Lifetime Mode=::NewEnumerator1
    Sprite Rotation Mode=::NewEnumerator2
    Sprite Size Mode=::NewEnumerator1
    Sprite UV Mode=::NewEnumerator3
    Uniform Sprite Size Max=75
    Uniform Sprite Size Min=35
    UsePositionOffset=false
    Write Material Random=true
   )
   &ShapeLocation(
    Handle Radius=4
    MaskSpawnGroup=false
    Non Uniform Scale=<0 16 30>
    Offset=<0 100 0>
    Offset Mode=::NewEnumerator0
    Rotation Mode=::NewEnumerator2
    Shape Primitive=::NewEnumerator3
    Torus Distribution Mode:ENiagaraTorusDistributionMode=::NewEnumerator1
    U Position=&MaskFloatBySpawnGroup(
     Masked Float=&FloatFromCurve(
      FloatCurve=DataInterface() {propertyValues {Curve {keys [<at 0 0.5>,<at 1 0.55>]}}}
      CurveIndex=&Subtract_Float(A=&Multiply_Float(A=$System.Age, B=$User.Speed), B=$User.Delay)
     )
     Passthrough Float=&FloatFromCurve(
      FloatCurve=DataInterface() {propertyValues {Curve {keys [<at 0 0.5>,<at 1.75 1>]}}}
      CurveIndex=&Subtract_Float(A=$System.Age, B=$User.Delay)
      Scale Curve=-1
     )
    )
    UseOwnerScale=true
    V Position=&RandomRangeFloat(Minimum=0.25, Maximum=0.75)
    Yaw / Pitch / Roll=<0 0.25 0.25>
   ) {module=ShapeLocation001}
   &AddVelocity(
    Module.Use Falloff Distance=false
    Override Seed=false
    Velocity Mode=::NewEnumerator1
    Velocity Speed=&RandomRangeFloat(Minimum=250, Maximum=300)
   ) {module=AddVelocity001}
   &AddVelocity(
    Velocity=$Particles.ShapeLocation001.ShapeNormal
    Velocity Speed Scale=&RandomRangeFloat(Minimum=500, Maximum=750)
   )
   /Niagara/Modules/Update/SubUV/SubUVAnimation(Number Of Frames=6, SubUV Animation Mode=::NewEnumerator1)
  ]
  particleUpdateScript [
   &ParticleState()
   &Drag(
    Drag=&FloatFromCurve(
     FloatCurve=DataInterface() {propertyValues {Curve {keys [<at 0 2 RCIM_Cubic>,<at 0.5 0>]}}}
     Scale Curve=&RandomRangeFloat(Minimum=2.5, Maximum=5)
    )
    Use Linear Drag=true
    Use Rotational Drag=true
   )
   &SpriteRotationRate(Rotation Rate=&Multiply_Float(
     A=&Lerp_Float(
      A=90
      B=45
      Alpha=&RemapRange(Input Value To Remap=$Particles.Mass, Input Min=0.75, Input Max=2, Output Min=0, Output Max=1)
     )
     B=&MakeFloatFromInt(Int=&RandomRangeInt(Minimum=-1, Maximum=1))
    )) {module=SpriteRotationRate001}
   &GravityForce(Gravity=<0 0 -920>)
   &SolveForcesAndVelocity(Clamp Velocity=false, Limit Acceleration=false)
   &ScaleColor(
    Scale RGB=&VectorFromCurve(VectorCurve=DataInterface() {propertyValues {
       XCurve {keys [<at 0 1.5 RCIM_Cubic>, {arriveTangent=-1.2988, interpMode=RCIM_Cubic, leaveTangent=-1.2988, tangentMode=RCTM_User, time=1, value=1}]}
       YCurve {keys [<at 0 1.5 RCIM_Cubic>, {arriveTangent=-1.2988, interpMode=RCIM_Cubic, leaveTangent=-1.2988, tangentMode=RCTM_User, time=1, value=1}]}
       ZCurve {keys [<at 0 1.5 RCIM_Cubic>, {arriveTangent=-1.2988, interpMode=RCIM_Cubic, leaveTangent=-1.2988, tangentMode=RCTM_User, time=1, value=1}]}
      }})
    ScaleA=true
    ScaleRGB=true
   )
   &ScaleSpriteSize(Uniform Curve Sprite Scale=DataInterface() {propertyValues {Curve {keys [
        {arriveTangent=0.8475, interpMode=RCIM_Cubic, leaveTangent=0.8475, tangentMode=RCTM_User, time=0, value=0}
        {arriveTangent=3.334, interpMode=RCIM_Cubic, leaveTangent=3.334, tangentMode=RCTM_User, time=0.3, value=1}
        <at 2 3 RCIM_Cubic>
       ]}}})
   &DynamicMaterialParameters(
    Index 0 Param 1=&FloatFromCurve(FloatCurve=DataInterface() {propertyValues {Curve {keys [
         {interpMode=RCIM_Cubic, tangentMode=RCTM_Break, time=0.5, value=1}
         {arriveTangent=-4.292, interpMode=RCIM_Cubic, leaveTangent=-4.292, tangentMode=RCTM_User, time=1, value=0}
        ]}}})
    Index 0 Param 2=&FloatFromCurve(
     FloatCurve=DataInterface() {propertyValues {Curve {keys [{arriveTangent=0.4361, interpMode=RCIM_Cubic, leaveTangent=0.4361, tangentMode=RCTM_User, time=0, value=0.05}, <at 1 0.25>]}}}
     Scale Curve=0.5
    )
    Index 1 Param 1=1
    Index 1 Param 2=1
    Index 1 Param 3=&RandomRangeFloat(Minimum=0, Maximum=1)
    Index 1 Param 4=&RandomRangeFloat(Minimum=0, Maximum=1)
    Module.Param0WriteEnabled=true
    Module.Param0WriteEnabled1=true
    Module.Param1WriteEnabled=true
    Module.Param1WriteEnabled1=true
    Module.Param2WriteEnabled=true
    Module.Param2WriteEnabled1=true
    Module.Param3WriteEnabled=true
    Module.Param3WriteEnabled1=true
    Write Parameter Index 1=true
   )
   &CameraOffset(Camera Offset Amount=&FloatFromCurve(FloatCurve=DataInterface() {propertyValues {Curve {keys [
         {interpMode=RCIM_Cubic, leaveTangent=1080.3575, tangentMode=RCTM_Break, time=0, value=15}
         {arriveTangent=135, interpMode=RCIM_Cubic, leaveTangent=135, tangentMode=RCTM_Break, time=0.12, value=94}
         <at 1 150>
        ]}}}))
   &SpriteRotationRate(Rotation Rate=&Multiply_Float(
     A=&Lerp_Float(A=-1, B=1, Alpha=$Particles.Initial.MaterialRandom)
     B=&Lerp_Float(A=125, B=0, Alpha=&Power(a=$Particles.NormalizedAge, b=0.2))
    ))
  ]
  rendererClasses [&NiagaraSpriteRendererProperties]
  renderers [{rendererClass=&NiagaraSpriteRendererProperties}]
 }
 Main_Gloop001 {
  bEnabled=false
  emitterUpdateScript [
   &EmitterState(
    Delay First Loop Only=true
    Enable Distance Culling=true
    Life Cycle Mode=::NewEnumerator1
    Loop Behavior=::NewEnumerator2
    Loop Count=6
    Loop Delay=$User.Delay
    Loop Duration=0.25
    MaxDistance=7000
    Module.bUseMaxDistance=true
    Module.bUseMinDistance=false
    Scalability Mode=::NewEnumerator1
    UseLoopDelay=true
   )
   &SpawnRate(
    SpawnRate=&FloatFromCurve(
     FloatCurve=DataInterface() {propertyValues {Curve {keys [<at 0 1>,<at 0.5 0.4>,<at 1 0>]}}}
     CurveIndex=&Subtract_Float(A=&Multiply_Float(A=$System.Age, B=$User.Speed), B=$User.Delay)
     Scale Curve=12.5
    )
    Use Spawn Probability=false
   ) {module=SpawnRate001}
   &SpawnRate(
    SpawnGroup=1
    SpawnRate=&FloatFromCurve(
     FloatCurve=DataInterface() {propertyValues {Curve {keys [<at 0 1>,<at 1.25 0>]}}}
     CurveIndex=&Subtract_Float(A=&Multiply_Float(A=$System.Age, B=$User.Speed), B=$User.Delay)
     Scale Curve=25
    )
    Use Spawn Probability=false
   ) {module=SpawnRate002}
  ]
  particleSpawnScript [
   /SetVariables(Particles.SpawnGroup=$Emitter.SpawnGroup) {isSetParameters=true}
   &InitializeParticle(
    Color=&SelectLinearColorFromArray(Array Sampling Mode=::NewEnumerator0, Direct Array Index=?, Color Selection Array=$User.Graffiti_Color_Array, Random Seed=0)
    Lifetime Max=0.7
    Lifetime Min=0.35
    Lifetime Mode=::NewEnumerator1
    Sprite Rotation Mode=::NewEnumerator2
    Sprite Size Mode=::NewEnumerator1
    Sprite UV Mode=::NewEnumerator3
    Uniform Sprite Size Max=225
    Uniform Sprite Size Min=175
    UsePositionOffset=false
    Write Material Random=true
   )
   &ShapeLocation(
    Handle Radius=4
    MaskSpawnGroup=false
    Non Uniform Scale=<0 16 30>
    Offset=<0 100 0>
    Offset Mode=::NewEnumerator0
    Rotation Mode=::NewEnumerator2
    Shape Primitive=::NewEnumerator3
    Torus Distribution Mode:ENiagaraTorusDistributionMode=::NewEnumerator1
    U Position=&MaskFloatBySpawnGroup(
     Masked Float=&FloatFromCurve(
      FloatCurve=DataInterface() {propertyValues {Curve {keys [<at 0 0.5>,<at 1 0.55>]}}}
      CurveIndex=&Subtract_Float(A=&Multiply_Float(A=$System.Age, B=$User.Speed), B=$User.Delay)
     )
     Passthrough Float=&FloatFromCurve(
      FloatCurve=DataInterface() {propertyValues {Curve {keys [<at 0 0.5>,<at 1.75 1>]}}}
      CurveIndex=&Subtract_Float(A=$System.Age, B=$User.Delay)
      Scale Curve=-1
     )
    )
    UseOwnerScale=true
    V Position=&RandomRangeFloat(Minimum=0.25, Maximum=0.75)
    Yaw / Pitch / Roll=<0 0.25 0.25>
   ) {module=ShapeLocation001}
   &AddVelocity(
    Module.Use Falloff Distance=false
    Override Seed=false
    Velocity Mode=::NewEnumerator1
    Velocity Speed=&RandomRangeFloat(Minimum=250, Maximum=500)
   ) {module=AddVelocity001}
   &AddVelocity(
    Velocity=$Particles.ShapeLocation001.ShapeNormal
    Velocity Speed Scale=&RandomRangeFloat(Minimum=500, Maximum=750)
   )
   &SubUVAnimation(Number Of Frames=6, SubUV Animation Mode=::NewEnumerator1)
  ]
  particleUpdateScript [
   &ParticleState()
   &Drag(
    Drag=&FloatFromCurve(
     FloatCurve=DataInterface() {propertyValues {Curve {keys [<at 0 2 RCIM_Cubic>,<at 0.5 0>]}}}
     Scale Curve=&RandomRangeFloat(Minimum=2.5, Maximum=5)
    )
    Use Linear Drag=true
    Use Rotational Drag=true
   )
   &GravityForce(Gravity=<0 0 -920>)
   &SolveForcesAndVelocity(Clamp Velocity=false, Limit Acceleration=false)
   &ScaleColor(
    Scale Alpha=&FloatFromCurve(
     FloatCurve=DataInterface() {propertyValues {Curve {keys [
         {interpMode=RCIM_Cubic, tangentMode=RCTM_User, time=0, value=0}
         {interpMode=RCIM_Cubic, tangentMode=RCTM_User, time=0.4, value=1}
         {tangentMode=RCTM_User, time=1, value=0}
        ]}}}
     Scale Curve=1.25
    )
    Scale RGB=&VectorFromCurve(
     VectorCurve=DataInterface() {propertyValues {
       XCurve {keys [<at 0 1.5 RCIM_Cubic>, {arriveTangent=-1.2988, interpMode=RCIM_Cubic, leaveTangent=-1.2988, tangentMode=RCTM_User, time=1, value=1}]}
       YCurve {keys [<at 0 1.5 RCIM_Cubic>, {arriveTangent=-1.2988, interpMode=RCIM_Cubic, leaveTangent=-1.2988, tangentMode=RCTM_User, time=1, value=1}]}
       ZCurve {keys [<at 0 1.5 RCIM_Cubic>, {arriveTangent=-1.2988, interpMode=RCIM_Cubic, leaveTangent=-1.2988, tangentMode=RCTM_User, time=1, value=1}]}
      }}
     Scale Curve=&VectorFromFloat(Value=0.5)
    )
    ScaleA=true
    ScaleRGB=true
   )
   &ScaleSpriteSize(Uniform Curve Sprite Scale=DataInterface() {propertyValues {Curve {keys [
        {arriveTangent=0.8475, interpMode=RCIM_Cubic, leaveTangent=0.8475, tangentMode=RCTM_User, time=0, value=0}
        {arriveTangent=3.334, interpMode=RCIM_Cubic, leaveTangent=3.334, tangentMode=RCTM_User, time=0.3, value=1}
        <at 2 3 RCIM_Cubic>
       ]}}})
   &DynamicMaterialParameters(
    Index 0 Param 1=&FloatFromCurve(FloatCurve=DataInterface() {propertyValues {Curve {keys [
         {arriveTangent=0.0912, interpMode=RCIM_Cubic, leaveTangent=0.0912, tangentMode=RCTM_User, time=0, value=1}
         {arriveTangent=0.2099, interpMode=RCIM_Cubic, leaveTangent=0.2099, tangentMode=RCTM_User, time=1.0011, value=0}
        ]}}})
    Module.Param0WriteEnabled=true
    Module.Param1WriteEnabled=true
    Module.Param2WriteEnabled=true
    Module.Param3WriteEnabled=true
   ) {module=DynamicMaterialParameters001}
   &CameraOffset(Camera Offset Amount=50)
   &SpriteRotationRate(Rotation Rate=&Lerp_Float(A=-150, B=150, Alpha=$Particles.Initial.MaterialRandom))
  ]
  rendererClasses [&NiagaraSpriteRendererProperties]
  renderers [{rendererClass=&NiagaraSpriteRendererProperties}]
 }
 PIC_FLIP_Fluid {
  bEnabled=false
  simTarget=GPUComputeSim
  emitterSpawnScript [
   /SetVariables(Emitter.RasterGrid=DataInterface() {propertyValues {
      ClearBeforeNonIterationStage=true
      NumAttributes=2
      NumCellsMaxAxis=3
      NumCellsX=128
      NumCellsY=64
      OverrideBufferFormat=Float
      PreviewAttribute=None
      SetGridFromMaxAxis=false
      bOverrideFormat=false
      bPreviewGrid=false
      RenderTargetUserParameter {parameter {name=None, typeDefHandle {registeredTypeIndex=4}, varData []}}
      WorldBBoxSize=<100 100>
     }}) {isSetParameters=true}
   /SetVariables(Emitter.VelocityGrid=DataInterface() {propertyValues {
      ClearBeforeNonIterationStage=true
      NumAttributes=0
      NumCellsMaxAxis=3
      NumCellsX=128
      NumCellsY=64
      OverrideBufferFormat=Float
      PreviewAttribute=None
      SetGridFromMaxAxis=false
      bOverrideFormat=true
      bPreviewGrid=false
      RenderTargetUserParameter {parameter {name=None, typeDefHandle {registeredTypeIndex=4}, varData []}}
      WorldBBoxSize=<100 100>
     }}) {isSetParameters=true}
   /NiagaraFluids/Modules/Grid2D/Grid2D_SetResolution(
    Grid:NiagaraDataInterfaceGrid2DCollection=$Emitter.VelocityGrid
    NumCellsMaxAxis=200
    Resolution Method:ENiagaraGrid2DResolution=::NewEnumerator1
    WorldGridExtents:Vector2f=<600 600>
   )
   &Grid2D_SetResolution(
    Grid:NiagaraDataInterfaceGrid2DCollection=$Emitter.RasterGrid
    OtherGrid:NiagaraDataInterfaceGrid2DCollection=$Emitter.VelocityGrid
    Resolution Method:ENiagaraGrid2DResolution=::NewEnumerator3
    WorldGridExtents:Vector2f=$Emitter.Grid2D_SetResolution.WorldGridExtents
   ) {module=Grid2D_SetResolution001}
   /SetVariables(Emitter.DeltaTime=0.0166) {isSetParameters=true}
   /SetVariables() {enabled=false, isSetParameters=true}
   /NiagaraFluids/Modules/Grid2D/Grid2D_InitRT(Grid:NiagaraDataInterfaceGrid2DCollection=$Emitter.VelocityGrid, SimRT:NiagaraDataInterfaceRenderTarget2D=$Emitter.SimRT) {enabled=false, module=InitRTGrid2D}
   /SetVariables() {isSetParameters=true}
   &self:NiagaraEmitter_3.NiagaraScratchPadContainer_0.ScratchModule_0(Grid:NiagaraDataInterfaceGrid2DCollection=$Emitter.VelocityGrid) {module=ScratchModule}
  ]
  emitterUpdateScript [
   /NiagaraFN/Modules/Scalability/SignificantInstancesCull(Cull Response:ENiagaraExecutionStateManagement=::KillImmediately, Num Instances:NiagaraInt32=$System.NumFluidInstances)
   &EmitterState(Scalability Mode=::NewEnumerator1)
   &SpawnBurst_Instantaneous(Spawn Count=200000, Use Spawn Probability=false) {enabled=false}
   &SpawnRate(Use Spawn Probability=false)
   /NiagaraFluids/Modules/Grid2D/Grid2D_UpdateSimAttributes(Grid:NiagaraDataInterfaceGrid2DCollection=$Emitter.VelocityGrid, WorldGridSize:Vector2f=$Emitter.Grid2D_SetResolution.WorldGridExtents) {module=UpdateSimAttributes}
   /NiagaraFluids/Modules/Grid2D/Grid2D_CreateUnitToWorldTransform(
    Camera:NiagaraDataInterfaceCamera=$Emitter.CameraQuery
    Camera Facing=true
    Local Pivot:Vector2f=<0.5 0.1>
    Offset:NiagaraPosition=$Engine.Owner.Position
    Rotation:Quat4f=&self:NiagaraEmitter_3.NiagaraScratchPadContainer_0.ScratchDynamicInput_01_1(Rotation:Quat4f=$Engine.Owner.Rotation)
    WorldGridExtents:Vector2f=$Emitter.Grid2D_SetResolution.WorldGridExtents
   )
   &SpawnParticlesFromOtherEmitter(
    Attribute Reader=DataInterface() {propertyValues {EmitterBinding {bindingMode=Other, emitterName=Source}}}
    Spawn Rate=750
    Spawn Rate Per Particle Cap=750
    Use Maximum Age=false
    Use Minimum Age=false
   )
   /SetVariables(Emitter.Position:NiagaraPosition=&ConvertVectorToPosition(
     Input Position=$Emitter.Grid2D_CreateUnitToWorldTransform.LocalToWorldTranslation
     Conversion Operation:ENiagara_LWCConvertVecToPos=::NewEnumerator0
    )) {isSetParameters=true}
   /NiagaraFluids/Modules/Grid2D/Grid2D_SetBounds(
    EmitterProperties=DataInterface() {propertyValues {EmitterBinding {bindingMode=Self, emitterName=None}}}
    LocalPivot:Vector2f=$Emitter.Grid2D_CreateUnitToWorldTransform.LocalPivot
    WorldGridExtents:Vector2f=$Emitter.Grid2D_SetResolution.WorldGridExtents
   )
  ]
  particleSpawnScript [
   /Niagara/Modules/Spawn/Initialization/InitializeParticle(
    Lifetime=&RandomRangeFloat(Minimum=0.25, Maximum=0.5)
    Module.Write Color=true
    Module.Write Lifetime=true
    Module.Write Mass=false
    Module.Write Material Random=true
    Module.Write Position=false
    Module.Write Scale=false
    Module.Write SpriteRotation=false
    Module.Write SpriteSize=true
    Position=&SimulationPosition()
    Sprite Size=&RandomRangeVector2D(Minimum=<2 2>, Maximum=<4 4>)
   )
   &SampleParticlesFromOtherEmitter(
    Color Sampling=::NewEnumerator0
    Material Random Sampling:ENiagara_DirectReadSamplingMode=::NewEnumerator0
    Velocity Sampling=::NewEnumerator0
   )
   /Niagara/Modules/Spawn/Location/SphereLocation(MaskSpawnGroup=false, Module.Override Randomness=false, Non Uniform Scale=<1 0 1>, Offset=<0 0 200>, Sphere Radius=1, Surface Only=false) {enabled=false, module=SphereLocation002}
   /SetVariables() {enabled=false, isSetParameters=true}
   /Niagara/Modules/Update/Forces/CurlNoiseForce(Noise Frequency=0.1, Noise Strength=150, Pan Noise=false)
   &ApplyInitialForces()
   /Niagara/Modules/Spawn/Location/V2/SphereLocation(
    MaskSpawnGroup=false
    Override Seed=false
    Sphere Radius=&Multiply_Float(
     A=&FloatFromCurve(FloatCurve=DataInterface() {propertyValues {Curve {keys [<at 0 1>,<at 1 0>]}}}, CurveIndex=$Emitter.NormalizedLoopAge, Scale Curve=2)
     B=&Lerp_Float(A=10, B=1, Alpha=$Particles.MaterialRandom)
    )
    Surface Only=false
    UseNonUniformScale=false
    UseOffset=false
    UseSphereOrientation=false
   )
  ]
  particleUpdateScript [
   &ParticleState()
   &GravityForce(Gravity=<0 0 -920>)
   &CurlNoiseForce(Noise Frequency=0.02, Noise Strength=250, Pan Noise=true, Pan Noise Field=<0.2 0.1 0.2>) {module=CurlNoiseForce001}
   /SetVariables(Transient.StartPosition:NiagaraPosition=$Particles.Position) {isSetParameters=true}
   &SolveForcesAndVelocity(Clamp Velocity=false, DeltaTime:NiagaraFloat=$Emitter.DeltaTime, Limit Acceleration=false)
   /SetVariables(Particles.Position:Vector3f=$Transient.StartPosition) {isSetParameters=true}
   &ScaleColor(ScaleA=true, ScaleRGB=true)
   &KillParticlesInVolume(Box Size:Vector3f=<1750 500 300>, Invert Volume=true, Kill Shape=::NewEnumerator1, Origin Offset:Vector3f=<600 0 100>)
  ]
  rendererClasses [&NiagaraSpriteRendererProperties]
  renderers [
   {rendererClass=&NiagaraSpriteRendererProperties}
   {rendererClass=&NiagaraSpriteRendererProperties, rendererIndex=1}
  ]
 }
 Smoke {
  emitterUpdateScript [
   &EmitterState()
   &SpawnBurst_Instantaneous(Spawn Count=6, Spawn Group=2, Use Spawn Probability=false, UseLoopCountLimit=false)
   &SpawnRate(
    SpawnRate=&FloatFromCurve(
     FloatCurve=DataInterface() {propertyValues {Curve {keys [<at 0 1>,<at 0.8 0>]}}}
     CurveIndex=&Subtract_Float(A=&Multiply_Float(A=$System.Age, B=$User.Speed), B=$User.Delay)
     Scale Curve=&Multiply_Float(A=75, B=$User.Scale_Rate)
    )
    Use Spawn Probability=false
   )
   &SpawnRate(
    SpawnGroup=1
    SpawnRate=&FloatFromCurve(
     FloatCurve=DataInterface() {propertyValues {Curve {keys [<at 0 1>,<at 1 0.4>,<at 1.35 0>]}}}
     CurveIndex=&Subtract_Float(A=&Multiply_Float(A=$System.Age, B=$User.Speed), B=$User.Delay)
     Scale Curve=&Multiply_Float(A=80, B=$User.Scale_Rate)
    )
    Use Spawn Probability=false
   ) {module=SpawnRate001}
  ]
  particleSpawnScript [
   /SetVariables(Particles.SpawnGroup=$Emitter.SpawnGroup) {isSetParameters=true}
   /Niagara/Modules/Spawn/Initialization/V2/InitializeParticle(
    Color=&SelectLinearColorFromArray(Array Sampling Mode=::NewEnumerator0, Direct Array Index=?, Color Selection Array=$User.Graffiti_Color_Array, Random Seed=0)
    Lifetime Max=0.75
    Lifetime Min=0.5
    Lifetime Mode=::NewEnumerator1
    Mass Mode=::NewEnumerator1
    Position=&Add_Vector(A=<0 0 0>, B=<30 0 0>)
    Position Mode=::NewEnumerator1
    Sprite Rotation Angle=&UniformRangedFloat(Minimum=0, Maximum=360)
    Sprite Rotation Mode=::NewEnumerator0
    Sprite Size=&Multiply_Vector2D_ByFloat(
     Float=0.85
     Vector2D=&Vector2DFromFloat(Value=&UniformRangedFloat(Minimum=250, Maximum=400))
    )
    Sprite Size Mode=::NewEnumerator4
    UsePositionOffset=false
    Write Material Random=true
   ) {module=InitializeParticle001}
   &Color(Color=&MultiplyLinearColor(
     Linear Color=$Particles.Color
     Scale Factor=&MakeVector4_FromLinearColor(LinearColor=&UniformRangedLinearColor(Minimum=<rgba 0.5 0.5 0.5 0.3>, Maximum=<rgba 1 1 1 0.5>, Color Channel Mode=::NewEnumerator2))
    ))
   &AddVelocity(
    Velocity=<0 0 3500>
    Velocity Speed Scale=&SelectFloatFromArray(Direct Array Index=$Particles.SpawnGroup, Float Selection Array=DataInterface() {propertyValues {FloatData [0,0,1]}})
   ) {module=AddVelocity001}
   &ShapeLocation(
    Handle Radius=4
    MaskSpawnGroup=false
    Non Uniform Scale=<0 14 40>
    Offset=<125 100 -25>
    Offset Mode=::NewEnumerator0
    Rotation Mode=::NewEnumerator2
    Shape Primitive=::NewEnumerator3
    Torus Distribution Mode:ENiagaraTorusDistributionMode=::NewEnumerator1
    U Position=&FloatFromCurve(
     FloatCurve=DataInterface() {propertyValues {Curve {keys [<at 0 0.5>,<at 1.75 0.825>]}}}
     CurveIndex=&Subtract_Float(A=&Multiply_Float(A=$System.Age, B=$User.Speed), B=$User.Delay)
     Scale Curve=&MaskFloatBySpawnGroup(
      Masked Float=0.91
      Passthrough Float=&MaskFloatBySpawnGroup(Masked Float=-1, Passthrough Float=0, Spawn Group Mask=1)
     )
    )
    UseOwnerScale=true
    V Position=&RandomRangeFloat(Minimum=0.25, Maximum=0.75)
    Yaw / Pitch / Roll=<0 0.25 0.25>
   ) {module=ShapeLocation001}
   &AddVelocity(
    Velocity=$Particles.ShapeLocation001.ShapeNormal
    Velocity Speed Scale=&RandomRangeFloat(Minimum=1000, Maximum=1250)
   )
   /SetVariables(Particles.UVScale:Vector2f=&UniformRangedVector2D(Minimum=<-1 1>, Maximum=<-1 1>)) {isSetParameters=true}
  ]
  particleUpdateScript [
   &ParticleState()
   &Drag(Drag=&UniformRangedFloat(Minimum=8, Maximum=16), Use Linear Drag=true, Use Rotational Drag=false)
   &GravityForce(Gravity=<0 0 -420>)
   &SolveForcesAndVelocity(Clamp Velocity=false, Limit Acceleration=false)
   &SpriteRotationRate(Rotation Rate=&Multiply_Float(
     A=&Lerp_Float(A=400, B=-400, Alpha=$Particles.Initial.MaterialRandom)
     B=&Lerp_Float(A=0.15, B=0.05, Alpha=$Particles.NormalizedAge)
    ))
   &ScaleColor(
    Scale Alpha=&FloatFromCurve(
     FloatCurve=DataInterface() {propertyValues {Curve {keys [
         {interpMode=RCIM_Cubic, tangentMode=RCTM_User, time=0.5, value=1}
         {arriveTangent=0.137, interpMode=RCIM_Cubic, leaveTangent=0.137, tangentMode=RCTM_User, time=1, value=0}
        ]}}}
     Scale Curve=1.25
    )
    Scale RGB=<1.25 1.25 1.25>
    ScaleA=true
    ScaleRGB=true
   )
   &ScaleSpriteSize(Scale Factor=&Vector2DFromFloat(Value=&Lerp_Float(A=1, B=2, Alpha=$Particles.NormalizedAge)))
   &DynamicMaterialParameters(
    Index 0 Param 2=5
    Index 0 Param 3=0.8
    Module.Index0Param0Write=true
    Module.Index0Param1Write=true
    Module.Index0Param2Write=true
    Module.Index0Param3Write=false
   )
   /Niagara/Modules/Update/SubUV/V2/SubUVAnimation(
    "Sprite Renderer "=DataInterface() {propertyValues {SpriteRenderer=&self:Smoke_0.NiagaraSpriteRendererProperties_0}}
    SubUV Animation Mode:ENiagara_SubUVLookupModeV2=::NewEnumerator1
    UseEndFrame=false
    UseStartFrame=false
   ) {module=SubUVAnimation001}
   &CameraOffset(Camera Offset Amount=15)
   &self:ColorFromTOD_02(
    Evening:LinearColor=<rgba 1 0.625 0.5 1>
    Morning:LinearColor=<rgba 1 0.75 0.7753 1>
    Night:LinearColor=<rgba 1 0.4991 0.6404 0.8>
    TOD:Vector4f=&MakeVector4_FromLinearColor(LinearColor=$NPC.FortniteNPC.FortniteActiveTimeOfDay)
   ) {module=ColorFromTOD_02}
  ]
  rendererClasses [&NiagaraSpriteRendererProperties]
  renderers [{rendererClass=&NiagaraSpriteRendererProperties}]
 }
 Source {
  bEnabled=false
  simTarget=GPUComputeSim
  emitterUpdateScript [
   &SignificantInstancesCull(Cull Response:ENiagaraExecutionStateManagement=::KillImmediately, Num Instances:NiagaraInt32=$System.NumFluidInstances)
   &EmitterState(
    Delay First Loop Only=true
    Life Cycle Mode=::NewEnumerator1
    Loop Behavior=::NewEnumerator2
    Loop Count=6
    Loop Delay=$User.Delay
    Loop Duration=0.25
    Scalability Mode=::NewEnumerator1
    UseLoopDelay=true
   )
   &SpawnRate(
    SpawnRate=&FloatFromCurve(
     FloatCurve=DataInterface() {propertyValues {Curve {keys [<at 0 1>,<at 0.75 0>]}}}
     CurveIndex=&Subtract_Float(A=&Multiply_Float(A=$System.Age, B=$User.Speed), B=$User.Delay)
     Scale Curve=100
    )
    Use Spawn Probability=false
   ) {module=SpawnRate001}
   &SpawnRate(
    SpawnGroup=1
    SpawnRate=&FloatFromCurve(
     FloatCurve=DataInterface() {propertyValues {Curve {keys [<at 0 1>,<at 1.25 0>]}}}
     CurveIndex=&Subtract_Float(A=&Multiply_Float(A=$System.Age, B=$User.Speed), B=$User.Delay)
     Scale Curve=125
    )
    Use Spawn Probability=false
   ) {module=SpawnRate002}
  ]
  particleSpawnScript [
   /SetVariables(Particles.SpawnGroup=$Emitter.SpawnGroup) {isSetParameters=true}
   &InitializeParticle(
    Color=&SelectLinearColorFromArray(Array Sampling Mode=::NewEnumerator0, Direct Array Index=?, Color Selection Array=$User.Graffiti_Color_Array, Random Seed=0)
    Lifetime=&RandomRangeFloat(Minimum=0.5, Maximum=0.75)
    Mass=&RandomRangeFloat(Minimum=0.75, Maximum=1.75)
    Mass Mode=::NewEnumerator1
    Sprite Size=&Multiply_Vector2D_ByFloat(
     Float=0
     Vector2D=&Vector2DFromFloat(Value=&RandomRangeFloat(Minimum=8, Maximum=15))
    )
    Sprite Size Mode=::NewEnumerator4
    UsePositionOffset=false
    Write Material Random=true
   ) {module=InitializeParticle002}
   &ShapeLocation(
    Handle Radius=4
    MaskSpawnGroup=false
    Non Uniform Scale=<0 16 30>
    Offset=<0 100 0>
    Offset Mode=::NewEnumerator0
    Rotation Mode=::NewEnumerator2
    Shape Primitive=::NewEnumerator3
    Torus Distribution Mode:ENiagaraTorusDistributionMode=::NewEnumerator1
    U Position=&MaskFloatBySpawnGroup(
     Masked Float=&FloatFromCurve(
      FloatCurve=DataInterface() {propertyValues {Curve {keys [<at 0 0.5>,<at 1 0.55>]}}}
      CurveIndex=&Subtract_Float(A=&Multiply_Float(A=$System.Age, B=$User.Speed), B=$User.Delay)
     )
     Passthrough Float=&FloatFromCurve(
      FloatCurve=DataInterface() {propertyValues {Curve {keys [<at 0 0.5>,<at 1.75 1>]}}}
      CurveIndex=&Subtract_Float(A=$System.Age, B=$User.Delay)
      Scale Curve=-1
     )
    )
    UseOwnerScale=true
    V Position=&RandomRangeFloat(Minimum=0.25, Maximum=0.75)
    Yaw / Pitch / Roll=<0 0.25 0.25>
   )
   &AddVelocity(
    Velocity=$Particles.ShapeLocation001.ShapeNormal
    Velocity Speed Scale=&RandomRangeFloat(Minimum=750, Maximum=1800)
   ) {module=AddVelocity005}
   &AddVelocity(
    Module.Use Falloff Distance=false
    Override Seed=false
    Velocity Mode=::NewEnumerator1
    Velocity Speed=&RandomRangeFloat(Minimum=500, Maximum=750)
   ) {module=AddVelocity004}
   &AddVelocity(Velocity=<0 0 1500>)
  ]
  particleUpdateScript [
   &ParticleState()
   &GravityForce(Gravity=<0 0 -1800>)
   &Drag(
    Drag=&FloatFromCurve(
     FloatCurve=DataInterface() {propertyValues {Curve {keys [<at 0 2 RCIM_Cubic>,<at 0.5 0>]}}}
     Scale Curve=&RandomRangeFloat(Minimum=5, Maximum=6)
    )
    Use Linear Drag=true
    Use Rotational Drag=false
   )
   &ScaleColor(
    Scale Alpha=&FloatFromCurve(FloatCurve=DataInterface() {propertyValues {Curve {keys [<at 0.4596 1.0113 RCIM_Cubic>,<at 1.0011 0.002 RCIM_Cubic>]}}})
    ScaleA=true
    ScaleRGB=true
   )
   &CurlNoiseForce(Noise Strength=230, Pan Noise=false)
   &SolveForcesAndVelocity(Clamp Velocity=false, Limit Acceleration=false)
  ]
  rendererClasses [&NiagaraSpriteRendererProperties]
  renderers [{rendererClass=&NiagaraSpriteRendererProperties}]
 }
 Splashes {
  bEnabled=false
  emitterUpdateScript [
   &EmitterState(
    Delay First Loop Only=true
    Life Cycle Mode=::NewEnumerator1
    Loop Behavior=::NewEnumerator2
    Loop Count=6
    Loop Delay=$User.Delay
    Loop Duration=0.25
    UseLoopDelay=true
   )
   &SpawnRate(
    SpawnRate=&FloatFromCurve(
     FloatCurve=DataInterface() {propertyValues {Curve {keys [<at 0 1>,<at 0.75 0>]}}}
     CurveIndex=&Subtract_Float(A=&Multiply_Float(A=$System.Age, B=$User.Speed), B=$User.Delay)
     Scale Curve=75
    )
    Use Spawn Probability=false
   )
   &SpawnRate(
    SpawnGroup=1
    SpawnRate=&FloatFromCurve(
     FloatCurve=DataInterface() {propertyValues {Curve {keys [<at 0 1>,<at 1.25 0>]}}}
     CurveIndex=&Subtract_Float(A=&Multiply_Float(A=$System.Age, B=$User.Speed), B=$User.Delay)
     Scale Curve=100
    )
    Use Spawn Probability=false
   ) {module=SpawnRate001}
  ]
  particleSpawnScript [
   /SetVariables(Particles.SpawnGroup=$Emitter.SpawnGroup) {isSetParameters=true}
   &InitializeParticle(
    Color Mode=::NewEnumerator0
    Lifetime Max=0.75
    Lifetime Min=0.5
    Lifetime Mode=::NewEnumerator1
    Mass Max=2
    Mass Mode=::NewEnumerator3
    Sprite Rotation Mode=::NewEnumerator2
    Sprite Size Mode=::NewEnumerator1
    Sprite UV Mode=::NewEnumerator3
    Uniform Sprite Size Max=100
    Uniform Sprite Size Min=75
    UsePositionOffset=false
    Write Material Random=true
   )
   &ShapeLocation(
    Handle Radius=4
    MaskSpawnGroup=false
    Non Uniform Scale=<0 16 30>
    Offset=<0 100 0>
    Offset Mode=::NewEnumerator0
    Rotation Mode=::NewEnumerator2
    Shape Primitive=::NewEnumerator3
    Torus Distribution Mode:ENiagaraTorusDistributionMode=::NewEnumerator1
    U Position=&MaskFloatBySpawnGroup(
     Masked Float=&FloatFromCurve(
      FloatCurve=DataInterface() {propertyValues {Curve {keys [<at 0 0.5>,<at 1 0.55>]}}}
      CurveIndex=&Subtract_Float(A=&Multiply_Float(A=$System.Age, B=$User.Speed), B=$User.Delay)
     )
     Passthrough Float=&FloatFromCurve(
      FloatCurve=DataInterface() {propertyValues {Curve {keys [<at 0 0.5>,<at 1.75 1>]}}}
      CurveIndex=&Subtract_Float(A=$System.Age, B=$User.Delay)
      Scale Curve=-1
     )
    )
    UseOwnerScale=true
    V Position=&RandomRangeFloat(Minimum=0.25, Maximum=0.75)
    Yaw / Pitch / Roll=<0 0.25 0.25>
   ) {module=ShapeLocation001}
   &AddVelocity(
    Module.Use Falloff Distance=false
    Override Seed=false
    Velocity Mode=::NewEnumerator1
    Velocity Speed=&RandomRangeFloat(Minimum=250, Maximum=300)
   ) {module=AddVelocity001}
   &AddVelocity(
    Velocity=$Particles.ShapeLocation001.ShapeNormal
    Velocity Speed Scale=&RandomRangeFloat(Minimum=500, Maximum=750)
   )
   &CameraOffset(Camera Offset Amount=50)
  ]
  particleUpdateScript [
   &ParticleState()
   &SpriteRotationRate(Rotation Rate=&Multiply_Float(
     A=&Lerp_Float(
      A=180
      B=45
      Alpha=&RemapRange(Input Value To Remap=$Particles.Mass, Input Min=0.75, Input Max=2, Output Min=0, Output Max=1)
     )
     B=&MakeFloatFromInt(Int=&RandomRangeInt(Minimum=-1, Maximum=1))
    ))
   &Drag(
    Drag=&FloatFromCurve(
     FloatCurve=DataInterface() {propertyValues {Curve {keys [<at 0 2 RCIM_Cubic>,<at 0.5 0>]}}}
     Scale Curve=&RandomRangeFloat(Minimum=2.5, Maximum=5)
    )
    Use Linear Drag=true
    Use Rotational Drag=true
   )
   &GravityForce(Gravity=<0 0 -920>)
   &SolveForcesAndVelocity(Clamp Velocity=false, Limit Acceleration=false)
   &ScaleSpriteSize(Uniform Curve Sprite Scale=DataInterface() {propertyValues {Curve {keys [
        {arriveTangent=1.8917, interpMode=RCIM_Cubic, leaveTangent=1.8917, tangentMode=RCTM_User, time=0, value=0.35}
        {arriveTangent=-0.5754, interpMode=RCIM_Cubic, leaveTangent=-0.5754, tangentMode=RCTM_User, time=1, value=2}
       ]}}})
   &ScaleColor(
    Scale Alpha=&FloatFromCurve(FloatCurve=DataInterface() {propertyValues {Curve {keys [
         <at 0 -1>
         <at 0.0002 0.2 RCIM_Cubic>
         {arriveTangent=-2.4995, interpMode=RCIM_Cubic, leaveTangent=-2.4995, tangentMode=RCTM_User, time=1, value=-1}
        ]}}})
    Scale RGB=&Multiply_VectorByFloat(
     Vector=&MakeVectorFromLinearColorRGB(Color=&MultiplyLinearColor(
       Linear Color=&ColorFromCurve(ColorCurve=DataInterface() {propertyValues {
          AlphaCurve {keys []}
          BlueCurve {keys [<at 0.1481 40>, {arriveTangent=-36.1948, interpMode=RCIM_Cubic, leaveTangent=-36.1948, time=0.198, value=12}, <at 0.9216 12>, <at 1 10>]}
          GreenCurve {keys [<at 0.1481 28.1816>, {arriveTangent=-20.9175, interpMode=RCIM_Cubic, leaveTangent=-20.9175, time=0.198, value=12}, <at 0.9216 12>, <at 1 10>]}
          RedCurve {keys [<at 0.1481 7.2711>, {arriveTangent=6.1129, interpMode=RCIM_Cubic, leaveTangent=6.1129, time=0.198, value=12}, <at 0.9216 12>, <at 1 10>]}
         }})
       Scale Factor=&MakeVector4_FromLinearColor(LinearColor=&SelectLinearColorFromArray(Array Sampling Mode=::NewEnumerator0, Direct Array Index=?, Color Selection Array=$User.Graffiti_Color_Array, Random Seed=0))
      ))
     Float=1.5
    )
    ScaleA=true
    ScaleRGB=true
   )
  ]
  rendererClasses [&NiagaraSpriteRendererProperties]
  renderers [{rendererClass=&NiagaraSpriteRendererProperties}]
 }
}
summary {
 systemName=NS_Train_Graffiti_Success_Capture_Box
 emitters [
  {emitterName=Smoke, rendererClasses [&NiagaraSpriteRendererProperties]}
  {bEnabled=false, emitterName=PIC_FLIP_Fluid, simTarget=GPUComputeSim, rendererClasses [&NiagaraSpriteRendererProperties]}
  {bEnabled=false, emitterName=Source, simTarget=GPUComputeSim, rendererClasses [&NiagaraSpriteRendererProperties]}
  {bEnabled=false, emitterName=Splashes, rendererClasses [&NiagaraSpriteRendererProperties]}
  {bEnabled=false, emitterName=Main_Gloop, rendererClasses [&NiagaraSpriteRendererProperties]}
  {bEnabled=false, emitterName=Glow, rendererClasses [&NiagaraSpriteRendererProperties]}
  {bEnabled=false, emitterName=Main_Gloop001, rendererClasses [&NiagaraSpriteRendererProperties]}
 ]
 userVariables [
  User.Delay {type=NiagaraFloat, defaultValue=0.4, description=""}
  User.Graffiti_Color_Array {
   type=NiagaraDataInterfaceArrayColor
   description=""
   defaultValue {
    struct=NiagaraExt_VariableValue_DataInterface
    value {
     dataInterface=&self:NiagaraDataInterfaceArrayColor_0
     dataInterfaceClass=&NiagaraDataInterfaceArrayColor
    }
   }
  }
  User.Scale_Rate {type=NiagaraFloat, defaultValue=0.45, description=""}
  User.Speed {type=NiagaraFloat, defaultValue=1, description=""}
 ]
}
