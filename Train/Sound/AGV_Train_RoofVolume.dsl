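; Audio gameplay volume for the train roof that drives the player-on-train mix and a line-shaped audio source along the cars.
; On cosmetic-capable instances, BeginPlay sizes and lifts the volume from the parent actor's bounds and gets the train manager.
; PlayerOnTrain and PlayerOffTrain set a control bus mix value and toggle ticking and the AudioShapeLineList.
; Tick rebuilds the line from raised points at each car's front plus the last car's rear.
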(event EventBeginPlay
  (bind _returnvalue (Actor|GetAttachParentActor))
  (bind (_origin _boxextent) (Collision|GetActorBounds _returnvalue true false))
  (Networking|CanExecuteCosmeticEvents
    (:True
      (Variables|Default|SetActorBoundsOrigin _origin)
      (Variables|Default|SetActorBoundsExtent _boxextent)
      (Transformation|SetWorldLocation (Variables|Default|GetScene) (+ (Variables|Default|GetActorBoundsOrigin) (Math|Vector|MakeVector 0.0 0.0 (* (.z (Variables|Default|GetActorBoundsExtent)) 2.0))) false false)
      (Transformation|SetWorldTransform (Variables|Default|GetScene) (Math|Transform|MakeTransform (Transformation|GetWorldLocation (Variables|Default|GetScene)) (Transformation|GetWorldRotation (Variables|Default|GetScene)) (* (Variables|Default|GetActorBoundsExtent) 2.1)) false false)
      (Variables|Default|SetTrainManager (Actor|GetComponentbyClass (Actor|GetAttachParentActor) "/Script/TrainCoreRuntime.FortTrainManager"))
      (bind _trainmanager (Variables|Default|GetTrainManager)
        (:then
          (Utilities|Array|Resize (Variables|Default|GetConnectionPoints) (+ (Utilities|Array|Length (Variables|/Script/TrainCoreRuntime.FortTrainManager|GetCars _trainmanager)) 1)))
        (:else)))))

(event EventTick
  (params (DeltaSeconds :type Float%20%28single-precision%29))
  (bind _returnvalue (Math|Vector|MakeVector 0.0 0.0 1.5e+02))
  (bind _connectionpoints (Variables|Default|GetConnectionPoints))
  (bind _trainmanager (Variables|Default|GetTrainManager)
    (:then
      (for (_array_element _array_index) (Variables|/Script/TrainCoreRuntime.FortTrainManager|GetCars _trainmanager)
        (Utilities|IsValid (Utilities|Struct|BreakTrainCarRuntimeData _array_element)
          (:Is%20Valid
            (Utilities|Array|SetArrayElem _connectionpoints _array_index (Utilities|Operators|Add (Transformation|GetActorLocation (Utilities|Struct|BreakTrainCarRuntimeData _array_element)) (* (Transformation|GetActorForwardVector (Utilities|Struct|BreakTrainCarRuntimeData _array_element)) (/ (.Length (Utilities|Struct|BreakTrainCarRuntimeData _array_element)) 2.0)) _returnvalue)))))
      (Utilities|IsValid (Utilities|Struct|BreakTrainCarRuntimeData (stale _array_element))
        (:Is%20Valid
          (Utilities|Array|SetArrayElem _connectionpoints (+ (stale _array_index) 1) (Utilities|Operators|Add (Transformation|GetActorLocation (Utilities|Struct|BreakTrainCarRuntimeData (stale _array_element))) (* (* (Transformation|GetActorForwardVector (Utilities|Struct|BreakTrainCarRuntimeData (stale _array_element))) (/ (.Length (Utilities|Struct|BreakTrainCarRuntimeData (stale _array_element))) 2.0)) -1.0) _returnvalue))
          (Variables|/Script/AudioShapes.AudioShapeLineListComponent|SetPointList _connectionpoints (Variables|Default|GetAudioShapeLineList)))))
    (:else)))

(event Custom|PlayerOnTrain
  (Audio|Modulation|SetGlobalControlBusMixValue "/Train/Sound/Settings/CB_PlayerIsOnTrain.CB_PlayerIsOnTrain" 1.0 0.5) ; Used to drive audio mix if player is on train
  (Actor|Tick|SetActorTickEnabled true)
  (Components|Activation|Activate (Variables|Default|GetAudioShapeLineList) false))

(event Custom|PlayerOffTrain
  (Audio|Modulation|SetGlobalControlBusMixValue "/Train/Sound/Settings/CB_PlayerIsOnTrain.CB_PlayerIsOnTrain" 0.0 0.5)
  (Actor|Tick|SetActorTickEnabled false)
  (Components|Activation|Deactivate (Variables|Default|GetAudioShapeLineList)))

(fn ConstructionScript)
