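; Train station PA controller that plays arrival, departure and random announcements for the station matching its identifier tag.
; Clients cache the station's DataTable sounds, track wireless speaker components via router events, and bind to train rail arrival-interval and departure events.
; The authority queues random announcement timers; playback checks listener range and enabled state, fading in the main audio alongside all speakers.
; Optionally, an audio gameplay volume toggles a control bus mix to suppress horns nearby.
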(event EventBeginPlay
  (bind _speakers (Variables|Default|GetSpeakers))
  (CallFunction|ShouldPreventHornatStation)
  (Utilities|FlowControl|Sequence
    (:then_0
      (Networking|SwitchHasAuthority
        (:Authority
          (Default|QueueRandomAnnouncement) ; Just going to queue up random announcements from the start
        )
        (:Remote
          (CallFunction|CacheStationData)
          (bind _returnvalue (GameplayEventRouter|GetEventRouter "/Script/FortniteGame.FortGameState" self))
          (Variables|Default|SetEventRouter _returnvalue)
          ;[0 0 0] Wait for Speakers to come online / get destroyed
          (bind (_asynctaskproxy _payload) (GameplayEventRouter|StartListeningToEvent (Variables|Default|GetEventRouter) "/Train/Sound/GameplayEvent_TrainStationSpeaker.GameplayEvent_TrainStationSpeaker")
            (:OnEventReceived
              (switch Utilities|FlowControl|Switch|SwitchonE_TrainStationSpeakerEventTypes (.EventType_5_9ECAE9514322D43A19BC17885671A60B (Utilities|Struct|BreakGameplayEventTrainStationSpeaker _payload))
                (:NewEnumerator0
                  (Utilities|Array|AddUnique _speakers (Utilities|Struct|BreakGameplayEventTrainStationSpeaker _payload)))
                (:NewEnumerator1
                  (Utilities|Array|RemoveItem _speakers (Utilities|Struct|BreakGameplayEventTrainStationSpeaker _payload)))))))))
    (:then_1
      (CallFunction|SetupRailEvents))))

(event EventEndPlay
  (params (EndPlayReason :type EEndPlayReason%20Enum :default "Destroyed"))
  (Utilities|Time|ClearandInvalidateTimerbyHandle (Variables|Default|GetResolveTrainRailTimer)))

(event Custom|SetupRailEvents
  (bind _trainrail (Variables|Default|GetTrainRail))
  (bind _output (Utilities|ResolveSoftReference _trainrail))
  ;[0 0 0] Setup rail delegates to get arrival time and departure events
  (Networking|SwitchHasAuthority
    (:Remote
      (Class|FortAsyncActionCurrentPlaylistReady|PlaylistDataReadyAsync
        (:OnReady
          (Variables|Default|SetTrainRailResolveAttempts 0)
          (if (Utilities|IsValidSoftObjectReference (Variables|Default|GetTrainRail))
            (EventDispatchers|BindEventtoOnTrainStationArrivalIntervalReachedEvent _output (EventDispatchers|CreateEvent :!event Custom|OnStationArrivalIntervalReached))
            (EventDispatchers|BindEventtoOnTrainStationDepartureEvent _output (EventDispatchers|CreateEvent :!event Custom|OnStationDepartureEvent))
            (else
              (bind _returnvalue (Utilities|Time|SetTimerbyEvent (delegate Custom|RetryResolveTrainRail) (Variables|Default|GetResolveTrainRailRetryTime)))
              (Variables|Default|SetResolveTrainRailTimer _returnvalue)
              (Math|Integer|IncrementInt (Variables|Default|GetTrainRailResolveAttempts)))))))))

(event Custom|RetryResolveTrainRail
  (bind _trainrail (Variables|Default|GetTrainRail))
  (bind _output (Utilities|ResolveSoftReference _trainrail))
  ;[0 0 0] Setup rail delegates to get arrival time and departure events
  (if (Utilities|IsValidSoftObjectReference (Variables|Default|GetTrainRail))
    (EventDispatchers|BindEventtoOnTrainStationArrivalIntervalReachedEvent _output (EventDispatchers|CreateEvent :!event Custom|OnStationArrivalIntervalReached))
    (EventDispatchers|BindEventtoOnTrainStationDepartureEvent _output (EventDispatchers|CreateEvent :!event Custom|OnStationDepartureEvent))
    (elif (< (Variables|Default|GetTrainRailResolveAttempts) 10)
      _
      (else
        (bind _returnvalue (Utilities|Time|SetTimerbyEvent (EventDispatchers|CreateEvent :!event Custom|RetryResolveTrainRail) (Variables|Default|GetResolveTrainRailRetryTime)))
        (Variables|Default|SetResolveTrainRailTimer _returnvalue)
        (Math|Integer|IncrementInt (Variables|Default|GetTrainRailResolveAttempts))))))

(event Custom|OnStationArrivalIntervalReached
  (params (Train :type Fort%20Train%20Manager%20Object%20Reference) (StationIdentifier :type Gameplay%20Tag%20Structure) (TimeUntilArrival :type Float%20%28single-precision%29) (Interval :type Float%20%28single-precision%29))
  (bind _currenttimeuntilarrival (Variables|Default|GetCurrentTimeUntilArrival))
  ;[0 0 0] Logic for triggering announcements
  (Networking|CanExecuteCosmeticEvents
    (:True
      (if (GameplayTags|MatchesTag StationIdentifier (Variables|Default|GetTrainStationIdentifier) true)
        (bind _returnvalue (Audio|AreAnyListenersWithinRange (Transformation|GetActorLocation) (Variables|Default|GetMinDistanceFromPlayerForAnnouncement)))
        (Utilities|FlowControl|Sequence
          (:then_0
            (if _returnvalue
              (bind _output_get (Variables|Default|SetCurrentTimeUntilArrival TimeUntilArrival))
              (CallFunction|PrintDebugLog _output_get)
              (Utilities|IsValid (CallFunction|GetSoundfromEventTag _output_get)
                (:Is%20Valid
                  (Default|PlayAnnouncement (CallFunction|GetSoundfromEventTag _output_get))))))
          (:then_1
            (if (!= _currenttimeuntilarrival 0)
              (Variables|Default|SetbRandomAnnouncementsEnabled (> _currenttimeuntilarrival 1.2e+02))
              (else
                (Default|SetRandomAnnouncementsEnabled true)))))))))

(event Custom|OnStationDepartureEvent
  (params (Train :type Fort%20Train%20Manager%20Object%20Reference) (StationIdentifier :type Gameplay%20Tag%20Structure) (TimeUntilNextArrival :type Float%20%28single-precision%29))
  ;[0 0 0] Logic for triggering announcements
  (if (GameplayTags|MatchesTag StationIdentifier (Variables|Default|GetTrainStationIdentifier) true)
    (CallFunction|PlayDepartingAnnouncement)))

(event Custom|PlayTest
  ;[0 0 0] Test event for editor
  (Default|PlayAnnouncement "/Train/Sound/Cues/Spot_Train_Station_PA_Arrival_120_Sec_Cue.Spot_Train_Station_PA_Arrival_120_Sec_Cue"))

(event Custom|OnPlayRandomAnnouncement
  (bind (_stationdata_stationtag_22_c0bed43b46661f4c5da4b79535592f04 _stationdata_soundatarrival_31_7117959e44b8268f91c964bb2e9ccc81 _stationdata_soundatrandom_18_c87945074c663a01a0731188205640c7) (Utilities|Struct|BreakTrainStationAudioData (Variables|GetStationData_StationTag_22_C0BED43B46661F4C5DA4B79535592F04)))
  ;[0 0 0] Random Announcements that get shut off a T-Minus 3min from arrival
  (Networking|SwitchHasAuthority
    (:Authority
      (Default|QueueRandomAnnouncement))
    (:Remote
      (bind _returnvalue (Audio|AreAnyListenersWithinRange (Transformation|GetActorLocation) (Variables|Default|GetMinDistanceFromPlayerForAnnouncement)))
      (if (Math|Boolean|ANDBoolean _returnvalue (Variables|Default|GetbRandomAnnouncementsEnabled) (> (- (Utilities|Time|GetGameTimeinSeconds) (Variables|Default|GetGameTimeWhenAnnouncementsEnabled)) 2e+01))
        (Default|PlayAnnouncement _stationdata_soundatrandom_18_c87945074c663a01a0731188205640c7)))))

(event OnAudioFinished(Audio)
  ;[0 0 0] When audio finishes stop all wireless speaker components
  (for _array_element (Variables|Default|GetSpeakers)
    (Default|Stop _array_element)))

(event Custom|ShouldPreventHornAtStation
  (bind _audiogameplayvolume (Variables|Default|GetAudioGameplayVolume))
  ;[0 0 0 0.5] Opt in AGV modulation setup to prevent train horn sounds to play nearby stations
  (Class|FortAsyncActionCurrentPlaylistReady|PlaylistDataReadyAsync
    (:OnReady
      (if (Variables|Default|GetbPreventHornAtStation)
        (Events|BindEventtoOnToggledOn _audiogameplayvolume (EventDispatchers|CreateEvent :!event OnPlayerEnterAGV))
        (Events|BindEventtoOnToggledOff _audiogameplayvolume (EventDispatchers|CreateEvent :!event OnPlayerExitAGV))
        (else
          (bind _audioshapesphere (Variables|Default|GetAudioShapeSphere)
            (:then
              (Components|DestroyComponent _audioshapesphere))
            (:else
              _))
          (bind _audiogameplayvolume_2 (Variables|Default|GetAudioGameplayVolume)
            (:then
              (Components|DestroyComponent _audiogameplayvolume_2))
            (:else)))))))

(fn CacheStationData
  (bind _outrownames (DataTable|GetDataTableRowNames (Variables|Default|GetDataTable)))
  (for _array_element _outrownames
    (bind _returnvalue (Utilities|GetDataTableRow (Variables|Default|GetDataTable) _array_element))
    (if (== (Variables|Default|GetTrainStationIdentifier) (Utilities|Struct|BreakTrainStationAudioData _returnvalue))
      (Variables|Default|SetStationData _returnvalue)
      (return true)))
  (return false))

(fn GetRandomAnnouncementTime
  (bind (_randomannouncementtimerange_x _randomannouncementtimerange_y) (Math|Vector2D|BreakVector2D (Variables|GetRandomAnnouncementTimeRange_X)))
  (return (Math|Random|RandomFloatinRange _randomannouncementtimerange_x _randomannouncementtimerange_y)))

(fn GetSoundFromEventTag
  (params (IntervalBeforeArriving :type Float%20%28double-precision%29))
  (bind (_stationdata_stationtag_22_c0bed43b46661f4c5da4b79535592f04 _stationdata_soundatarrival_31_7117959e44b8268f91c964bb2e9ccc81 _stationdata_soundatrandom_18_c87945074c663a01a0731188205640c7 _stationdata_soundat90s_39_941284ce41d4715d3388b6bc79ba25a1 _stationdata_soundat60s_26_5ce48920435525cbdeeeff8f317db012 _stationdata_soundat30s_28_9876acb34d6bede7a4cc04aed2867968 _stationdata_soundatarrivingloop_38_71b84c31472bf2dff611dbbd7980b700) (Utilities|Struct|BreakTrainStationAudioData (Variables|GetStationData_StationTag_22_C0BED43B46661F4C5DA4B79535592F04)))
  (Utilities|FlowControl|Sequence
    (:then_0
      (switch string (Utilities|String|ToString(Integer) (Math|Float|Truncate IntervalBeforeArriving))
        (:0
          (LocalVariables|SetSoundToPlayLocal _stationdata_soundatarrival_31_7117959e44b8268f91c964bb2e9ccc81))
        (:90
          (LocalVariables|SetSoundToPlayLocal _stationdata_soundat90s_39_941284ce41d4715d3388b6bc79ba25a1))
        (:60
          (LocalVariables|SetSoundToPlayLocal _stationdata_soundat60s_26_5ce48920435525cbdeeeff8f317db012))
        (:30
          (LocalVariables|SetSoundToPlayLocal _stationdata_soundat30s_28_9876acb34d6bede7a4cc04aed2867968))
        (:15
          (LocalVariables|SetSoundToPlayLocal _stationdata_soundatarrivingloop_38_71b84c31472bf2dff611dbbd7980b700))))
    (:then_1
      (return (LocalVariables|GetSoundToPlayLocal)))))

(fn OnAnnouncementFinished)

(fn OnPlayerEnterAGV
  (Audio|Modulation|ActivateControlBusMix (Variables|Default|GetHornCBM)))

(fn OnPlayerExitAGV
  (Audio|Modulation|DeactivateControlBusMix (Variables|Default|GetHornCBM)))

(fn PlayAnnouncement
  (params (Sound :type Sound%20Base%20Object%20Reference))
  (bind _audio (Variables|Default|GetAudio))
  (Audio|Components|Audio|SetSound _audio Sound)
  (Audio|Components|Audio|FadeIn _audio 0.1 1.0 0.0 "Linear" :!class /Script/Engine.AudioComponent)
  (for _array_element (Variables|Default|GetSpeakers)
    (Class|BPTrainStationSpeakerComponent|Play _array_element)))

(fn PlayDepartingAnnouncement
  (bind (_stationdata_stationtag_22_c0bed43b46661f4c5da4b79535592f04 _stationdata_soundatarrival_31_7117959e44b8268f91c964bb2e9ccc81 _stationdata_soundatrandom_18_c87945074c663a01a0731188205640c7 _stationdata_soundat90s_39_941284ce41d4715d3388b6bc79ba25a1 _stationdata_soundat60s_26_5ce48920435525cbdeeeff8f317db012 _stationdata_soundat30s_28_9876acb34d6bede7a4cc04aed2867968 _stationdata_soundatarrivingloop_38_71b84c31472bf2dff611dbbd7980b700 _stationdata_soundatdeparture_37_eba28a3345a42f66df4a6aa6ed8ea1f8) (Utilities|Struct|BreakTrainStationAudioData (Variables|GetStationData_StationTag_22_C0BED43B46661F4C5DA4B79535592F04)))
  (CallFunction|PrintDebugLog (Variables|Default|GetCurrentTimeUntilArrival))
  (Default|PlayAnnouncement _stationdata_soundatdeparture_37_eba28a3345a42f66df4a6aa6ed8ea1f8)
  (Default|SetRandomAnnouncementsEnabled true))

(fn PrintDebugLog
  (params (TimeUntilArrival :type Float%20%28double-precision%29))
  (if (Variables|Default|GetbDebugEnabled)
    (Fortnite|FortShippingLog self (Utilities|String|ToString(Text) (Utilities|Text|FormatText "NSLOCTEXT(\"[812B5A7FE022DF0CF08A615302F4DD84]\", \"C001220A4721F6BF4CCA60822AE707D6\", \"TrainStationEvent: Tag = {TAG}, Event = {EVENT}, NumSpeakers = {NUM}\")" (GameplayTags|GetDebugStringfromGameplayTag (Variables|Default|GetTrainStationIdentifier)) (Utilities|String|ToString(Float) TimeUntilArrival) (Utilities|Array|Length (Variables|Default|GetSpeakers)))))))

(fn QueueRandomAnnouncement
  (bind _returnvalue (Utilities|Time|SetTimerbyEvent (EventDispatchers|CreateEvent :!event OnPlayRandomAnnouncement) (CallFunction|GetRandomAnnouncementTime)))
  (Variables|Default|SetRandomAnnouncementTimer _returnvalue))

(fn SetRandomAnnouncementsEnabled
  (params (Enabled :type Boolean))
  (bind _output_get (Variables|Default|SetbRandomAnnouncementsEnabled Enabled))
  (if _output_get
    (Variables|Default|SetGameTimeWhenAnnouncementsEnabled (Utilities|Time|GetGameTimeinSeconds))))

(fn ConstructionScript)
