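; Train station speaker component that registers with a listening controller and plays or stops a sound through its owner's audio component.
; On clients running cosmetic events, BeginPlay caches that component, waits for game state readiness, then broadcasts a speaker gameplay event through the game state.
; EndPlay broadcasts the same event with a different state value.
; Play fades in the configured sound; Stop either fires an OnStop trigger parameter or stops playback outright.
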(event EventBeginPlay
  (Networking|CanExecuteCosmeticEvents
    (:True
      (CallFunction|CacheAudioComponent)
      (Class|FortAsyncActionGameStateReady|GameStateReadyAsync
        (:OnReady
          ;[0 0 0] Broadcast out for our controller to listen
          (GameplayEvent|BroadcastEvent (Variables|Default|GetSpeakerEventDesc) 0 (Utilities|Struct|MakeGameplayEventTrainStationSpeaker self) "/Script/FortniteGame.FortGameState"))))))

(event EventEndPlay
  (params (EndPlayReason :type EEndPlayReason%20Enum :default "Destroyed"))
  (Networking|CanExecuteCosmeticEvents
    (:True
      ;[0 0 0] Broadcast out for our controller to listen
      (GameplayEvent|BroadcastEvent (Variables|Default|GetSpeakerEventDesc) 0 (Utilities|Struct|MakeGameplayEventTrainStationSpeaker self "NewEnumerator1") "/Script/FortniteGame.FortGameState"))))

(fn CacheAudioComponent
  (Variables|Default|SetAudioComponent (Actor|GetComponentbyClass (Components|GetOwner) "/Script/Engine.AudioComponent")))

(fn Play
  (bind _audiocomponent (Variables|Default|GetAudioComponent)
    (:then
      (Audio|Components|Audio|SetSound _audiocomponent (Variables|Default|GetSound))
      (Audio|Components|Audio|FadeIn _audiocomponent 0.1 1.0 0.0 "Linear" :!class /Script/Engine.AudioComponent))
    (:else)))

(fn Stop
  (bind _audiocomponent (Variables|Default|GetAudioComponent)
    (:then
      (if (Variables|Default|GetbSendStopTrigger)
        (Audio|Parameter|ExecuteTriggerParameter _audiocomponent "OnStop")
        (else
          (Audio|Components|Audio|Stop _audiocomponent))))
    (:else)))
